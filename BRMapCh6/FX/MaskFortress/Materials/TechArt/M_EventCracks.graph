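# Unlit translucent deferred decal material that draws glowing, animated cracks, likely for a Mask Fortress event.
# An outer crack texture supplies opacity and a bump-offset parallax height, while an inverted, power-sharpened inner crack texture builds the color mask.
# Named reroutes organize color, emissive and opacity, alongside a cosine time pulse and two multiplied panning noise samples.
# Both noise panners use only the red channel of their SpeedXY parameters, so they scroll diagonally.
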Material /BRMapCh6/FX/MaskFortress/Materials/TechArt/M_EventCracks
graph {
 properties {
  bAutomaticallySetUsageInEditor=false
  blendMode=BLEND_Translucent
  materialDomain=MD_DeferredDecal
  shadingModel=MSM_Unlit
  translucencyLightingMode=TLM_VolumetricDirectional
  usedShadingModels=Unlit
 }
 nodes [
  Constant()
  NamedRerouteDeclaration(None=LinearInterpolate(
    A=VectorParameter() {parameterName=Color3 - Dark, defaultValue=<rgba 0 0 0 1>}.RGB
    B=VectorParameter() {parameterName=Color1 - Bright, defaultValue=<rgba 0.3802 0 0.238 1>}.RGB
    Alpha=NamedRerouteUsage()
   ) {constAlpha=1}) {name=OUT_Color, nodeColor=<rgba 1 0.9844 0 1>}
  NamedRerouteDeclaration(None=TextureSample(UVs=NamedRerouteUsage()) {samplerType=SAMPLERTYPE_LinearGrayscale, texture=&/T_Cracks_Outer}.R) {name=OUT_Opacity, nodeColor=<rgba 0.6172 1 0 1>}
  n1 = ScalarParameter() {defaultValue=1, parameterName=Emissive - A}
  n2 = ScalarParameter() {defaultValue=1, parameterName=Emissive - B}
  n3 = NamedRerouteUsage()
  NamedRerouteDeclaration(None=Multiply(A=NamedRerouteUsage(), B=LinearInterpolate(A=@n1, B=@n2, Alpha=@n3))) {name=OUT_Emissive, nodeColor=<rgba 1 0 0.1406 1>}
  n4 = OneMinus(None=TextureSample(UVs=NamedRerouteUsage()) {samplerType=SAMPLERTYPE_LinearGrayscale, texture=&/T_Cracks_Inner_Hermes}.R)
  n5 = LinearInterpolate(
   A=ScalarParameter() {defaultValue=10, parameterName=Power - A}
   B=ScalarParameter() {defaultValue=10, parameterName=Power - B}
   Alpha=NamedRerouteUsage()
  )
  NamedRerouteDeclaration(None=Multiply(A=Power(Base=@n4, Exp=@n5) {constExponent=10}, B=NamedRerouteUsage())) {name=ColorMask, nodeColor=<rgba 0 0.7422 1 1>}
  n6 = ScalarParameter() {parameterName=Lowestvalue}
  n7 = ScalarParameter() {defaultValue=1, parameterName=Highestvalue}
  n8 = Divide(A=Time(), B=ScalarParameter() {defaultValue=10, parameterName=Time-Speed}) {constB=5}
  n9 = Add(A=Multiply(A=Cosine(None=@n8) {period=0.2}) {constB=-0.5}) {constB=0.5}
  NamedRerouteDeclaration(None=Saturate(None=LinearInterpolate(A=@n6, B=@n7, Alpha=@n9))) {name=Pulse, nodeColor=<rgba 0 1 0.5313 1>}
  n10 = Power(
   Base=TextureSample() {samplerType=SAMPLERTYPE_LinearGrayscale, texture=&T_Cracks_Outer}.RGB
   Exp=ScalarParameter() {defaultValue=2, parameterName=Bump - Power}
  ) {constExponent=10}
  NamedRerouteDeclaration(None=BumpOffset(
    Coordinate=TextureCoordinate()
    Height=Saturate(None=@n10)
    HeightRatioInput=ScalarParameter() {defaultValue=0.05, parameterName=HeightRatioInput}
   )) {name=Parallax, nodeColor=<rgba 0 0.9531 1 1>}
  n11 = VectorParameter() {parameterName=Noise1 - SpeedXY, defaultValue=<rgba -0.39 0.56 0 1>}
  n12 = Panner(
   Coordinate=Multiply(A=TextureCoordinate(), B=ScalarParameter() {defaultValue=1, parameterName=Noise1 - Tiling}) {constB=2.17}
   Speed=AppendVector(A=@n11.R, B=@n11.R)
  )
  n13 = VectorParameter() {parameterName=Noise2 - SpeedXY, defaultValue=<rgba 0.86 -0.34 0 1>}
  n14 = Panner(
   Coordinate=Multiply(A=TextureCoordinate(), B=ScalarParameter() {defaultValue=2, parameterName=Noise2 - Tiling}) {constB=1.33}
   Speed=AppendVector(A=@n13.R, B=@n13.R)
  )
  NamedRerouteDeclaration(None=Multiply(
    A=TextureSample(UVs=@n12) {texture=&/PistolCanary_NoiseClouds}.RGB
    B=TextureSample(UVs=@n14) {texture=&PistolCanary_NoiseClouds}.RGB
   )) {name=MovingNoise, nodeColor=<rgba 0 1 0.1797 1>}
  NamedRerouteUsage()
  NamedRerouteUsage()
  NamedRerouteUsage()
 ]
}
