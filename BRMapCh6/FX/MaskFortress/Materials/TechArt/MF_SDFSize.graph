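# Material function in the MaskFortress FX tech art folder that remaps an SDF input into a clamped mask controlled by a Size input.
# It subtracts Size from the SDF, divides by the gap between Size and a fixed 0.9 ceiling, then saturates.
# Raising Size toward that ceiling narrows the ramp, giving a harder threshold.
# An "SDF Size - Main Shape" scalar parameter sits in the graph unconnected.
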MaterialFunction /BRMapCh6/FX/MaskFortress/Materials/TechArt/MF_SDFSize
graph {nodes [
  n1 = FunctionInput(Preview=Constant() {r=0.89}) {bUsePreviewValueAsDefault=true, inputName="Size (Max 0.9)", inputType=FunctionInput_Scalar}
  n2 = Subtract(A=FunctionInput(Preview=Constant() {r=0.9}) {bUsePreviewValueAsDefault=true, inputName=SDF, inputType=FunctionInput_Scalar}, B=@n1)
  n3 = Subtract(A=Constant() {r=0.9}, B=@n1)
  FunctionOutput(None=Saturate(None=Divide(A=@n2, B=@n3)))
  ScalarParameter() {defaultValue=0.8943, group=Main Shape, parameterName=SDF Size - Main Shape}
 ]}
