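# M_Glyphs is an effects material for the Chapter 6 Mask Fortress, apparently rendering glowing glyphs with layered depth.
# SDF sizing functions, radial distance masks and polar-coordinate UVs shape the patterns.
# A custom height-march loop, bump offsets and world-aligned textures projected behind translucency add parallax depth, with screen-pixel dither patterns.
# Platform switches select variants per hardware tier, and inverse eye adaptation compensates brightness for exposure.
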Material /BRMapCh6/FX/MaskFortress/Materials/TechArt/M_Glyphs
graph {
 expressions {
  MaterialExpressionActorPositionWS_0 {class=MaterialExpressionActorPositionWS, inputs [], properties {originType=CameraRelative}}
  MaterialExpressionActorPositionWS_1 {class=MaterialExpressionActorPositionWS, inputs [], properties {originType=CameraRelative}}
  MaterialExpressionActorPositionWS_3 {class=MaterialExpressionActorPositionWS, inputs [], properties {originType=CameraRelative}}
  MaterialExpressionAdd_0 {
   class=MaterialExpressionAdd
   inputs [
    A {from=MaterialExpressionMaterialFunctionCall_3, output=B}
    B {from=MaterialExpressionScalarParameter_5}
   ]
  }
  MaterialExpressionAdd_1 {
   class=MaterialExpressionAdd
   inputs [
    A {from=MaterialExpressionTextureCoordinate_3}
    B {from=""}
   ]
   properties {constB=-0.5}
  }
  MaterialExpressionAdd_10 {
   class=MaterialExpressionAdd
   inputs [
    A {from=MaterialExpressionNamedRerouteUsage_30}
    B {from=MaterialExpressionNamedRerouteUsage_40}
   ]
  }
  MaterialExpressionAdd_11 {
   class=MaterialExpressionAdd
   inputs [
    A {from=MaterialExpressionMultiply_41}
    B {from=MaterialExpressionStaticSwitchParameter_8}
   ]
  }
  MaterialExpressionAdd_12 {
   class=MaterialExpressionAdd
   inputs [
    A {from=MaterialExpressionMaterialFunctionCall_5, output=R}
    B {from=MaterialExpressionMaterialFunctionCall_5, output=R}
   ]
  }
  MaterialExpressionAdd_13 {
   class=MaterialExpressionAdd
   inputs [
    A {from=MaterialExpressionMaterialFunctionCall_15, output=Result}
    B {from=""}
   ]
   properties {constB=0.5}
  }
  MaterialExpressionAdd_14 {
   class=MaterialExpressionAdd
   inputs [
    A {from=MaterialExpressionMaterialFunctionCall_15, output=Result}
    B {from=""}
   ]
   properties {constB=0.75}
  }
  MaterialExpressionAdd_15 {
   class=MaterialExpressionAdd
   inputs [
    A {from=MaterialExpressionScalarParameter_17}
    B {from=MaterialExpressionConvert_3}
   ]
  }
  MaterialExpressionAdd_16 {
   class=MaterialExpressionAdd
   inputs [
    A {from=MaterialExpressionMultiply_31}
    B {from=MaterialExpressionMultiply_34}
   ]
   properties {constB=-2}
  }
  MaterialExpressionAdd_17 {
   class=MaterialExpressionAdd
   inputs [
    A {from=MaterialExpressionStaticSwitchParameter_14}
    B {from=""}
   ]
   properties {constB=-0.5}
  }
  MaterialExpressionAdd_18 {
   class=MaterialExpressionAdd
   inputs [
    A {from=MaterialExpressionMaterialFunctionCall_4, output=R}
    B {from=MaterialExpressionMaterialFunctionCall_4, output=R}
   ]
  }
  MaterialExpressionAdd_19 {
   class=MaterialExpressionAdd
   inputs [
    A {from=MaterialExpressionMultiply_33}
    B {from=MaterialExpressionMultiply_45}
   ]
   properties {constB=-2}
  }
  MaterialExpressionAdd_2 {
   class=MaterialExpressionAdd
   inputs [
    A {from=MaterialExpressionMaterialFunctionCall_15, output=Result}
    B {from=""}
   ]
   properties {constB=0.25}
  }
  MaterialExpressionAdd_20 {
   class=MaterialExpressionAdd
   inputs [
    A {from=MaterialExpressionConvert_3}
    B {from=MaterialExpressionScalarParameter_22}
   ]
  }
  MaterialExpressionAdd_21 {
   class=MaterialExpressionAdd
   inputs [
    A {from=MaterialExpressionScalarParameter_7}
    B {from=MaterialExpressionMultiply_23}
   ]
  }
  MaterialExpressionAdd_22 {
   class=MaterialExpressionAdd
   inputs [
    A {from=MaterialExpressionComponentMask_8}
    B {from=""}
   ]
   properties {constB=0.5}
  }
  MaterialExpressionAdd_23 {
   class=MaterialExpressionAdd
   inputs [
    A {from=MaterialExpressionMultiply_15}
    B {from=MaterialExpressionScalarParameter_44}
   ]
  }
  MaterialExpressionAdd_24 {
   class=MaterialExpressionAdd
   inputs [
    A {from=MaterialExpressionScalarParameter_47}
    B {from=MaterialExpressionConvert_7}
   ]
  }
  MaterialExpressionAdd_25 {
   class=MaterialExpressionAdd
   inputs [
    A {from=MaterialExpressionTextureCoordinate_1}
    B {from=""}
   ]
   properties {constB=-0.49}
  }
  MaterialExpressionAdd_26 {
   class=MaterialExpressionAdd
   inputs [
    A {from=MaterialExpressionMultiply_18}
    B {from=MaterialExpressionOneMinus_0}
   ]
  }
  MaterialExpressionAdd_27 {
   class=MaterialExpressionAdd
   inputs [
    A {from=MaterialExpressionTextureCoordinate_0}
    B {from=""}
   ]
   properties {constB=-0.5}
  }
  MaterialExpressionAdd_28 {
   class=MaterialExpressionAdd
   inputs [
    A {from=MaterialExpressionReroute_11}
    B {from=MaterialExpressionMultiply_36}
   ]
  }
  MaterialExpressionAdd_29 {
   class=MaterialExpressionAdd
   inputs [
    A {from=MaterialExpressionMultiply_50}
    B {from=MaterialExpressionLinearInterpolate_1}
   ]
  }
  MaterialExpressionAdd_3 {
   class=MaterialExpressionAdd
   inputs [
    A {from=MaterialExpressionNamedRerouteUsage_14}
    B {from=""}
   ]
  }
  MaterialExpressionAdd_30 {
   class=MaterialExpressionAdd
   inputs [
    A {from=MaterialExpressionAdd_3}
    B {from=MaterialExpressionMultiply_25}
   ]
  }
  MaterialExpressionAdd_32 {
   class=MaterialExpressionAdd
   inputs [
    A {from=MaterialExpressionConvert_7}
    B {from=MaterialExpressionScalarParameter_48}
   ]
  }
  MaterialExpressionAdd_33 {
   class=MaterialExpressionAdd
   inputs [
    A {from=MaterialExpressionNamedRerouteDeclaration_7}
    B {from=MaterialExpressionTransform_4}
   ]
  }
  MaterialExpressionAdd_34 {
   class=MaterialExpressionAdd
   inputs [
    A {from=MaterialExpressionScalarParameter_53}
    B {from=MaterialExpressionConvert_9}
   ]
  }
  MaterialExpressionAdd_35 {
   class=MaterialExpressionAdd
   inputs [
    A {from=MaterialExpressionConvert_9}
    B {from=MaterialExpressionScalarParameter_54}
   ]
  }
  MaterialExpressionAdd_36 {
   class=MaterialExpressionAdd
   inputs [
    A {from=MaterialExpressionMultiply_13}
    B {from=MaterialExpressionMultiply_21}
   ]
  }
  MaterialExpressionAdd_37 {
   class=MaterialExpressionAdd
   inputs [
    A {from=MaterialExpressionScreenPosition_3, output=PixelPosition}
    B {from=MaterialExpressionAppendVector_1}
   ]
  }
  MaterialExpressionAdd_38 {
   class=MaterialExpressionAdd
   inputs [
    A {from=MaterialExpressionNamedRerouteUsage_48}
    B {from=""}
   ]
   properties {constB=0.1}
  }
  MaterialExpressionAdd_4 {
   class=MaterialExpressionAdd
   inputs [
    A {from=""}
    B {from=""}
   ]
  }
  MaterialExpressionAdd_49 {
   class=MaterialExpressionAdd
   inputs [
    A {from=MaterialExpressionMaterialFunctionCall_25, output=Result}
    B {from=MaterialExpressionVectorParameter_6, output=RGB}
   ]
  }
  MaterialExpressionAdd_5 {
   class=MaterialExpressionAdd
   inputs [
    A {from=MaterialExpressionSaturate_2}
    B {from=MaterialExpressionSaturate_3}
   ]
  }
  MaterialExpressionAdd_50 {
   class=MaterialExpressionAdd
   inputs [
    A {from=MaterialExpressionMultiply_83}
    B {from=MaterialExpressionMultiply_84}
   ]
  }
  MaterialExpressionAdd_51 {
   class=MaterialExpressionAdd
   inputs [
    A {from=MaterialExpressionComponentMask_10}
    B {from=""}
   ]
   properties {constB=0.5}
  }
  MaterialExpressionAdd_52 {
   class=MaterialExpressionAdd
   inputs [
    A {from=MaterialExpressionScreenPosition_1, output=PixelPosition}
    B {from=MaterialExpressionAppendVector_5}
   ]
  }
  MaterialExpressionAdd_53 {
   class=MaterialExpressionAdd
   inputs [
    A {from=MaterialExpressionMax_6}
    B {from=MaterialExpressionMultiply_88}
   ]
  }
  MaterialExpressionAdd_54 {
   class=MaterialExpressionAdd
   inputs [
    A {from=MaterialExpressionSceneDepth_5, output=R}
    B {from=MaterialExpressionMultiply_89}
   ]
  }
  MaterialExpressionAdd_55 {
   class=MaterialExpressionAdd
   inputs [
    A {from=MaterialExpressionAdd_56}
    B {from=MaterialExpressionMultiply_91}
   ]
  }
  MaterialExpressionAdd_56 {
   class=MaterialExpressionAdd
   inputs [
    A {from=MaterialExpressionTextureCoordinate_10}
    B {from=MaterialExpressionMultiply_92}
   ]
  }
  MaterialExpressionAdd_57 {
   class=MaterialExpressionAdd
   inputs [
    A {from=MaterialExpressionTextureCoordinate_13}
    B {from=""}
   ]
   properties {constB=-0.5}
  }
  MaterialExpressionAdd_58 {
   class=MaterialExpressionAdd
   inputs [
    A {from=MaterialExpressionSmoothStep_9}
    B {from=MaterialExpressionOneMinus_12}
   ]
  }
  MaterialExpressionAdd_6 {
   class=MaterialExpressionAdd
   inputs [
    A {from=MaterialExpressionMultiply_51}
    B {from=MaterialExpressionAdd_38}
   ]
  }
  MaterialExpressionAdd_7 {
   class=MaterialExpressionAdd
   inputs [
    A {from=MaterialExpressionMaterialFunctionCall_11, output=Result}
    B {from=MaterialExpressionVectorParameter_7, output=RGB}
   ]
  }
  MaterialExpressionAdd_8 {
   class=MaterialExpressionAdd
   inputs [
    A {from=MaterialExpressionNamedRerouteUsage_6}
    B {from=MaterialExpressionMultiply_10}
   ]
  }
  MaterialExpressionAppendVector_0 {
   class=MaterialExpressionAppendVector
   inputs [
    A {from=MaterialExpressionConvert_8, output=""}
    B {from=MaterialExpressionConstant_11, output=""}
   ]
  }
  MaterialExpressionAppendVector_1 {
   class=MaterialExpressionAppendVector
   inputs [
    A {from=MaterialExpressionViewProperty_0, output=Property}
    B {from=MaterialExpressionViewProperty_0, output=Property}
   ]
  }
  MaterialExpressionAppendVector_3 {
   class=MaterialExpressionAppendVector
   inputs [
    A {from=MaterialExpressionMultiply_29, output=""}
    B {from=MaterialExpressionMultiply_37, output=""}
   ]
  }
  MaterialExpressionAppendVector_4 {
   class=MaterialExpressionAppendVector
   inputs [
    A {from=MaterialExpressionStaticSwitchParameter_0, output=""}
    B {from=MaterialExpressionConstant_7, output=""}
   ]
  }
  MaterialExpressionAppendVector_5 {
   class=MaterialExpressionAppendVector
   inputs [
    A {from=MaterialExpressionViewProperty_2, output=Property}
    B {from=MaterialExpressionViewProperty_2, output=Property}
   ]
  }
  MaterialExpressionBumpOffset_3 {
   class=MaterialExpressionBumpOffset
   inputs [
    Coordinate {from=MaterialExpressionNamedRerouteDeclaration_5, output=""}
    Height {from=MaterialExpressionConstant_9, output=""}
    HeightRatioInput {from="", output=""}
   ]
   properties {}
  }
  MaterialExpressionCameraVectorWS_4 {class=MaterialExpressionCameraVectorWS, inputs [], properties {}}
  MaterialExpressionCameraVectorWS_5 {class=MaterialExpressionCameraVectorWS, inputs [], properties {}}
  MaterialExpressionComponentMask_0 {
   class=MaterialExpressionComponentMask
   inputs [None {from=MaterialExpressionMaterialFunctionCall_8, output=Z Diffuse Texture}]
   properties {a=true}
  }
  MaterialExpressionComponentMask_10 {class=MaterialExpressionComponentMask, inputs [None {from=MaterialExpressionTransform_3}], properties {g=true, r=true}}
  MaterialExpressionComponentMask_11 {class=MaterialExpressionComponentMask, inputs [None {from=MaterialExpressionCustom_2, output=return}], properties {r=true}}
  MaterialExpressionComponentMask_7 {
   class=MaterialExpressionComponentMask
   inputs [None {from=MaterialExpressionMaterialFunctionCall_24, output=Z Diffuse Texture}]
   properties {a=true}
  }
  MaterialExpressionComponentMask_8 {class=MaterialExpressionComponentMask, inputs [None {from=MaterialExpressionTransform_0}], properties {g=true, r=true}}
  MaterialExpressionComponentMask_9 {class=MaterialExpressionComponentMask, inputs [None {from=MaterialExpressionAdd_33}], properties {g=true, r=true}}
  MaterialExpressionConstantBiasScale_0 {class=MaterialExpressionConstantBiasScale, inputs [None {from=MaterialExpressionCustom_3}], properties {bias=-2, scale=0.1}}
  MaterialExpressionConstantBiasScale_3 {class=MaterialExpressionConstantBiasScale, inputs [None {from=MaterialExpressionTextureSample_9, output=R}], properties {bias=-0.5, scale=2}}
  MaterialExpressionConstantBiasScale_4 {class=MaterialExpressionConstantBiasScale, inputs [None {from=MaterialExpressionCustom_4}], properties {bias=-2, scale=0.1}}
  MaterialExpressionConstant_0 {class=MaterialExpressionConstant}
  MaterialExpressionConstant_1 {class=MaterialExpressionConstant}
  MaterialExpressionConstant_10 {class=MaterialExpressionConstant}
  MaterialExpressionConstant_11 {class=MaterialExpressionConstant}
  MaterialExpressionConstant_12 {class=MaterialExpressionConstant, properties {r=0.3864}}
  MaterialExpressionConstant_13 {class=MaterialExpressionConstant, properties {r=8.5179}}
  MaterialExpressionConstant_14 {class=MaterialExpressionConstant}
  MaterialExpressionConstant_15 {class=MaterialExpressionConstant, properties {r=1}}
  MaterialExpressionConstant_16 {class=MaterialExpressionConstant}
  MaterialExpressionConstant_17 {class=MaterialExpressionConstant, properties {r=1.1508}}
  MaterialExpressionConstant_18 {class=MaterialExpressionConstant, properties {r=12.9822}}
  MaterialExpressionConstant_2 {class=MaterialExpressionConstant, properties {r=1}}
  MaterialExpressionConstant_29 {class=MaterialExpressionConstant}
  MaterialExpressionConstant_3 {class=MaterialExpressionConstant}
  MaterialExpressionConstant_30 {class=MaterialExpressionConstant, properties {r=1}}
  MaterialExpressionConstant_31 {class=MaterialExpressionConstant, properties {r=-1}}
  MaterialExpressionConstant_32 {class=MaterialExpressionConstant, properties {r=0.5}}
  MaterialExpressionConstant_33 {class=MaterialExpressionConstant, properties {r=0.5}}
  MaterialExpressionConstant_34 {class=MaterialExpressionConstant, properties {r=8}}
  MaterialExpressionConstant_35 {class=MaterialExpressionConstant}
  MaterialExpressionConstant_36 {class=MaterialExpressionConstant, properties {r=8}}
  MaterialExpressionConstant_4 {class=MaterialExpressionConstant, properties {r=1}}
  MaterialExpressionConstant_5 {class=MaterialExpressionConstant, properties {r=0.4}}
  MaterialExpressionConstant_6 {class=MaterialExpressionConstant}
  MaterialExpressionConstant_7 {class=MaterialExpressionConstant}
  MaterialExpressionConstant_8 {class=MaterialExpressionConstant}
  MaterialExpressionConstant_9 {class=MaterialExpressionConstant, properties {r=-3}}
  MaterialExpressionConvert_0 {
   class=MaterialExpressionConvert
   inputs [None {from=MaterialExpressionTextureCoordinate_6}]
   properties {
    nodeName=Break Float2
    convertInputs [{
      type=Vector2
      expressionInput {
       expression=&self:MaterialExpressionTextureCoordinate_6
       inputName=None
       mask=0
       maskA=0
       maskB=0
       maskG=0
       maskR=0
       outputIndex=0
      }
     }]
    convertOutputs [{type=Scalar}, {type=Scalar}]
   }
  }
  MaterialExpressionConvert_1 {
   class=MaterialExpressionConvert
   inputs [
    None {from=MaterialExpressionAdd_21}
    None {from=MaterialExpressionAdd_23}
   ]
   properties {
    nodeName=Make Float2
    convertInputs [
     {
      type=Scalar
      expressionInput {
       expression=&self:MaterialExpressionAdd_21
       inputName=None
       mask=0
       maskA=0
       maskB=0
       maskG=0
       maskR=0
       outputIndex=0
      }
     }
     {
      type=Scalar
      expressionInput {
       expression=&self:MaterialExpressionAdd_23
       inputName=None
       mask=0
       maskA=0
       maskB=0
       maskG=0
       maskR=0
       outputIndex=0
      }
     }
    ]
    convertOutputs [{type=Vector2}]
   }
  }
  MaterialExpressionConvert_2 {
   class=MaterialExpressionConvert
   inputs [
    None {from=MaterialExpressionMultiply_43}
    None {from=MaterialExpressionMultiply_44}
   ]
   properties {
    nodeName=Make Float2
    convertInputs [
     {
      type=Scalar
      expressionInput {
       expression=&self:MaterialExpressionMultiply_43
       inputName=None
       mask=0
       maskA=0
       maskB=0
       maskG=0
       maskR=0
       outputIndex=0
      }
     }
     {
      type=Scalar
      expressionInput {
       expression=&self:MaterialExpressionMultiply_44
       inputName=None
       mask=0
       maskA=0
       maskB=0
       maskG=0
       maskR=0
       outputIndex=0
      }
     }
    ]
    convertOutputs [{type=Vector2}]
   }
  }
  MaterialExpressionConvert_3 {
   class=MaterialExpressionConvert
   inputs [None {from=MaterialExpressionTextureCoordinate_7}]
   properties {
    nodeName=Break Float2
    convertInputs [{
      type=Vector2
      expressionInput {
       expression=&self:MaterialExpressionTextureCoordinate_7
       inputName=None
       mask=0
       maskA=0
       maskB=0
       maskG=0
       maskR=0
       outputIndex=0
      }
     }]
    convertOutputs [{type=Scalar}, {type=Scalar}]
   }
  }
  MaterialExpressionConvert_4 {
   class=MaterialExpressionConvert
   inputs [
    None {from=MaterialExpressionMaterialFunctionCall_3, output=R}
    None {from=MaterialExpressionMaterialFunctionCall_3, output=R}
    None {from=MaterialExpressionAdd_0}
   ]
   properties {
    nodeName=Make Float3
    convertInputs [
     {
      type=Scalar
      expressionInput {
       expression=&self:MaterialExpressionMaterialFunctionCall_3
       inputName=None
       mask=0
       maskA=0
       maskB=0
       maskG=0
       maskR=0
       outputIndex=0
      }
     }
     {
      type=Scalar
      expressionInput {
       expression=&self:MaterialExpressionMaterialFunctionCall_3
       inputName=None
       mask=0
       maskA=0
       maskB=0
       maskG=0
       maskR=0
       outputIndex=1
      }
     }
     {
      type=Scalar
      expressionInput {
       expression=&self:MaterialExpressionAdd_0
       inputName=None
       mask=0
       maskA=0
       maskB=0
       maskG=0
       maskR=0
       outputIndex=0
      }
     }
    ]
    convertOutputs [{type=Vector3}]
   }
  }
  MaterialExpressionConvert_6 {
   class=MaterialExpressionConvert
   inputs [
    None {from=MaterialExpressionMultiply_8}
    None {from=MaterialExpressionMultiply_11}
   ]
   properties {
    nodeName=Make Float2
    convertInputs [
     {
      type=Scalar
      expressionInput {
       expression=&self:MaterialExpressionMultiply_8
       inputName=None
       mask=0
       maskA=0
       maskB=0
       maskG=0
       maskR=0
       outputIndex=0
      }
     }
     {
      type=Scalar
      expressionInput {
       expression=&self:MaterialExpressionMultiply_11
       inputName=None
       mask=0
       maskA=0
       maskB=0
       maskG=0
       maskR=0
       outputIndex=0
      }
     }
    ]
    convertOutputs [{type=Vector2}]
   }
  }
  MaterialExpressionConvert_7 {
   class=MaterialExpressionConvert
   inputs [None {from=MaterialExpressionAdd_22}]
   properties {
    nodeName=Break Float2
    convertInputs [{
      type=Vector2
      expressionInput {
       expression=&self:MaterialExpressionAdd_22
       inputName=None
       mask=0
       maskA=0
       maskB=0
       maskG=0
       maskR=0
       outputIndex=0
      }
     }]
    convertOutputs [{type=Scalar}, {type=Scalar}]
   }
  }
  MaterialExpressionConvert_8 {
   class=MaterialExpressionConvert
   inputs [
    None {from=MaterialExpressionMultiply_27}
    None {from=MaterialExpressionMultiply_38}
   ]
   properties {
    nodeName=Make Float2
    convertInputs [
     {
      type=Scalar
      expressionInput {
       expression=&self:MaterialExpressionMultiply_27
       inputName=None
       mask=0
       maskA=0
       maskB=0
       maskG=0
       maskR=0
       outputIndex=0
      }
     }
     {
      type=Scalar
      expressionInput {
       expression=&self:MaterialExpressionMultiply_38
       inputName=None
       mask=0
       maskA=0
       maskB=0
       maskG=0
       maskR=0
       outputIndex=0
      }
     }
    ]
    convertOutputs [{type=Vector2}]
   }
  }
  MaterialExpressionConvert_9 {
   class=MaterialExpressionConvert
   inputs [None {from=MaterialExpressionTextureCoordinate_5}]
   properties {
    nodeName=Break Float2
    convertInputs [{
      type=Vector2
      expressionInput {
       expression=&self:MaterialExpressionTextureCoordinate_5
       inputName=None
       mask=0
       maskA=0
       maskB=0
       maskG=0
       maskR=0
       outputIndex=0
      }
     }]
    convertOutputs [{type=Scalar}, {type=Scalar}]
   }
  }
  MaterialExpressionCustom_2 {
   class=MaterialExpressionCustom
   inputs [
    tex {from=MaterialExpressionTextureObject_1, output=""}
    UVs {from=MaterialExpressionNamedRerouteUsage_49, output=""}
    StepSize {from=MaterialExpressionAdd_53, output=""}
    StepCount {from=MaterialExpressionScalarParameter_100, output=""}
    Distortion {from=MaterialExpressionMultiply_87, output=""}
    ViewOffset {from=MaterialExpressionNamedRerouteUsage_50, output=""}
    DDX {from=MaterialExpressionDDX_1, output=""}
    DDY {from=MaterialExpressionDDY_1, output=""}
   ]
   properties {
    code="float Height = Texture2DSample(tex, texSampler, UVs);\r\nfloat ScaledStep = Height * StepSize;\r\n\r\nfor (int i = 0; i <= StepCount; i++)\r\n{\r\n\tif (Height <= i / StepCount)\r\n\t{\r\n\tOutUV1 = UVs;\r\n\treturn Height * i / StepCount;\r\n\t}\r\n\r\n\tUVs += ViewOffset * ScaledStep;\r\n\tHeight = Texture2DSampleGrad(tex, texSampler, UVs + (Distortion * i / StepCount), DDX, DDY);\r\n}\r\nOutUV1 = UVs;\r\nreturn Height;"
    outputType=CMOT_Float1
    additionalOutputs [{outputName=OutUV1, outputType=CMOT_Float2}]
    inputs [
     {
      inputName=tex
      input {
       expression=&self:MaterialExpressionTextureObject_1
       inputName=None
       mask=0
       maskA=0
       maskB=0
       maskG=0
       maskR=0
       outputIndex=0
      }
     }
     {
      inputName=UVs
      input {
       expression=&self:MaterialExpressionNamedRerouteUsage_49
       inputName=None
       mask=0
       maskA=0
       maskB=0
       maskG=0
       maskR=0
       outputIndex=0
      }
     }
     {
      inputName=StepSize
      input {
       expression=&self:MaterialExpressionAdd_53
       inputName=None
       mask=0
       maskA=0
       maskB=0
       maskG=0
       maskR=0
       outputIndex=0
      }
     }
     {
      inputName=StepCount
      input {
       expression=&self:MaterialExpressionScalarParameter_100
       inputName=None
       mask=0
       maskA=0
       maskB=0
       maskG=0
       maskR=0
       outputIndex=0
      }
     }
     {
      inputName=Distortion
      input {
       expression=&self:MaterialExpressionMultiply_87
       inputName=None
       mask=0
       maskA=0
       maskB=0
       maskG=0
       maskR=0
       outputIndex=0
      }
     }
     {
      inputName=ViewOffset
      input {
       expression=&self:MaterialExpressionNamedRerouteUsage_50
       inputName=None
       mask=0
       maskA=0
       maskB=0
       maskG=0
       maskR=0
       outputIndex=0
      }
     }
     {
      inputName=DDX
      input {
       expression=&self:MaterialExpressionDDX_1
       inputName=None
       mask=0
       maskA=0
       maskB=0
       maskG=0
       maskR=0
       outputIndex=0
      }
     }
     {
      inputName=DDY
      input {
       expression=&self:MaterialExpressionDDY_1
       inputName=None
       mask=0
       maskA=0
       maskB=0
       maskG=0
       maskR=0
       outputIndex=0
      }
     }
    ]
   }
  }
  MaterialExpressionCustom_3 {
   class=MaterialExpressionCustom
   inputs [p {from=MaterialExpressionAdd_37, output=""}]
   properties {
    code="Mod( ((uint)(p.x) + 2 * (uint)(p.y)) , 5 )"
    outputType=CMOT_Float1
    inputs [{
      inputName=p
      input {
       expression=&self:MaterialExpressionAdd_37
       inputName=None
       mask=0
       maskA=0
       maskB=0
       maskG=0
       maskR=0
       outputIndex=0
      }
     }]
   }
  }
  MaterialExpressionCustom_4 {
   class=MaterialExpressionCustom
   inputs [p {from=MaterialExpressionAdd_52, output=""}]
   properties {
    code="Mod( ((uint)(p.x) + 2 * (uint)(p.y)) , 5 )"
    outputType=CMOT_Float1
    inputs [{
      inputName=p
      input {
       expression=&self:MaterialExpressionAdd_52
       inputName=None
       mask=0
       maskA=0
       maskB=0
       maskG=0
       maskR=0
       outputIndex=0
      }
     }]
   }
  }
  MaterialExpressionDDX_1 {class=MaterialExpressionDDX, inputs [Value {from=MaterialExpressionTextureCoordinate_11, output=""}], properties {}}
  MaterialExpressionDDY_1 {class=MaterialExpressionDDY, inputs [Value {from=MaterialExpressionTextureCoordinate_11, output=""}], properties {}}
  MaterialExpressionDistance_0 {
   class=MaterialExpressionDistance
   inputs [
    A {from=MaterialExpressionAdd_27, output=""}
    B {from=MaterialExpressionConstant_3, output=""}
   ]
   properties {}
  }
  MaterialExpressionDistance_2 {
   class=MaterialExpressionDistance
   inputs [
    A {from=MaterialExpressionAdd_25, output=""}
    B {from=MaterialExpressionConstant_16, output=""}
   ]
   properties {}
  }
  MaterialExpressionDistance_3 {
   class=MaterialExpressionDistance
   inputs [
    A {from=MaterialExpressionAdd_57, output=""}
    B {from=MaterialExpressionConstant_35, output=""}
   ]
   properties {}
  }
  MaterialExpressionDivide_14 {
   class=MaterialExpressionDivide
   inputs [
    A {from=MaterialExpressionSceneDepth_0, output=R}
    B {from=MaterialExpressionPixelDepth_0, output=R}
   ]
  }
  MaterialExpressionDivide_16 {
   class=MaterialExpressionDivide
   inputs [
    A {from=MaterialExpressionSubtract_20}
    B {from=""}
   ]
   properties {constB=100}
  }
  MaterialExpressionDivide_18 {
   class=MaterialExpressionDivide
   inputs [
    A {from=MaterialExpressionMultiply_16}
    B {from=MaterialExpressionNamedRerouteUsage_13}
   ]
  }
  MaterialExpressionDivide_19 {
   class=MaterialExpressionDivide
   inputs [
    A {from=MaterialExpressionScalarParameter_49}
    B {from=""}
   ]
   properties {constB=100}
  }
  MaterialExpressionDivide_4 {
   class=MaterialExpressionDivide
   inputs [
    A {from=MaterialExpressionTime_0}
    B {from=MaterialExpressionScalarParameter_21}
   ]
   properties {constB=3}
  }
  MaterialExpressionDivide_5 {
   class=MaterialExpressionDivide
   inputs [
    A {from=MaterialExpressionSubtract_1}
    B {from=""}
   ]
   properties {constB=100}
  }
  MaterialExpressionDivide_6 {
   class=MaterialExpressionDivide
   inputs [
    A {from=MaterialExpressionSceneDepth_4, output=R}
    B {from=MaterialExpressionPixelDepth_2, output=R}
   ]
  }
  MaterialExpressionDivide_7 {
   class=MaterialExpressionDivide
   inputs [
    A {from=MaterialExpressionScalarParameter_101}
    B {from=""}
   ]
  }
  MaterialExpressionDivide_8 {
   class=MaterialExpressionDivide
   inputs [
    A {from=MaterialExpressionTime_4}
    B {from=MaterialExpressionScalarParameter_110}
   ]
   properties {constB=3}
  }
  MaterialExpressionDotProduct_4 {
   class=MaterialExpressionDotProduct
   inputs [
    A {from=MaterialExpressionCameraVectorWS_4, output=""}
    B {from=MaterialExpressionVertexNormalWS_4, output=""}
   ]
   properties {}
  }
  MaterialExpressionEyeAdaptationInverse_0 {
   class=MaterialExpressionEyeAdaptationInverse
   inputs [
    LightValueInput {from=MaterialExpressionMultiply_40, output=""}
    AlphaInput {from=MaterialExpressionConstant_5, output=""}
   ]
   properties {}
  }
  MaterialExpressionEyeAdaptationInverse_1 {
   class=MaterialExpressionEyeAdaptationInverse
   inputs [
    LightValueInput {from=MaterialExpressionMultiply_93, output=""}
    AlphaInput {from=MaterialExpressionScalarParameter_60, output=""}
   ]
   properties {}
  }
  MaterialExpressionLinearInterpolate_0 {
   class=MaterialExpressionLinearInterpolate
   inputs [
    A {from=MaterialExpressionNamedRerouteUsage_55}
    B {from=MaterialExpressionNamedRerouteUsage_29}
    Alpha {from=MaterialExpressionSaturate_11}
   ]
  }
  MaterialExpressionLinearInterpolate_1 {
   class=MaterialExpressionLinearInterpolate
   inputs [
    A {from=MaterialExpressionVectorParameter_3, output=RGB}
    B {from=MaterialExpressionAdd_8}
    Alpha {from=MaterialExpressionSaturate_6}
   ]
  }
  MaterialExpressionLinearInterpolate_10 {
   class=MaterialExpressionLinearInterpolate
   inputs [
    A {from=""}
    B {from=""}
    Alpha {from=""}
   ]
  }
  MaterialExpressionLinearInterpolate_11 {
   class=MaterialExpressionLinearInterpolate
   inputs [
    A {from=MaterialExpressionReroute_3}
    B {from=MaterialExpressionMultiply_39}
    Alpha {from=MaterialExpressionSaturate_21}
   ]
  }
  MaterialExpressionLinearInterpolate_12 {
   class=MaterialExpressionLinearInterpolate
   inputs [
    A {from=MaterialExpressionConstant_17}
    B {from=MaterialExpressionMaterialFunctionCall_1, output=Result}
    Alpha {from=MaterialExpressionSaturate_10}
   ]
   properties {constA=1}
  }
  MaterialExpressionLinearInterpolate_13 {
   class=MaterialExpressionLinearInterpolate
   inputs [
    A {from=""}
    B {from=""}
    Alpha {from=MaterialExpressionSaturate_15}
   ]
   properties {constA=0.1}
  }
  MaterialExpressionLinearInterpolate_15 {
   class=MaterialExpressionLinearInterpolate
   inputs [
    A {from=MaterialExpressionMultiply_26}
    B {from=MaterialExpressionLinearInterpolate_3}
    Alpha {from=MaterialExpressionConstant_4}
   ]
  }
  MaterialExpressionLinearInterpolate_2 {
   class=MaterialExpressionLinearInterpolate
   inputs [
    A {from=MaterialExpressionReroute_6}
    B {from=MaterialExpressionMultiply_0}
    Alpha {from=MaterialExpressionReroute_7}
   ]
  }
  MaterialExpressionLinearInterpolate_3 {
   class=MaterialExpressionLinearInterpolate
   inputs [
    A {from=MaterialExpressionMultiply_54}
    B {from=MaterialExpressionNamedRerouteUsage_38}
    Alpha {from=MaterialExpressionStaticSwitchParameter_1}
   ]
  }
  MaterialExpressionLinearInterpolate_4 {
   class=MaterialExpressionLinearInterpolate
   inputs [
    A {from=MaterialExpressionNamedRerouteUsage_17}
    B {from=MaterialExpressionNamedRerouteUsage_18}
    Alpha {from=MaterialExpressionAdd_2}
   ]
   properties {constAlpha=0.25}
  }
  MaterialExpressionLinearInterpolate_5 {
   class=MaterialExpressionLinearInterpolate
   inputs [
    A {from=MaterialExpressionNamedRerouteUsage_19}
    B {from=MaterialExpressionNamedRerouteUsage_20}
    Alpha {from=MaterialExpressionAdd_13}
   ]
  }
  MaterialExpressionLinearInterpolate_6 {
   class=MaterialExpressionLinearInterpolate
   inputs [
    A {from=MaterialExpressionNamedRerouteUsage_21}
    B {from=MaterialExpressionNamedRerouteUsage_22}
    Alpha {from=MaterialExpressionAdd_14}
   ]
   properties {constAlpha=0.75}
  }
  MaterialExpressionLinearInterpolate_7 {
   class=MaterialExpressionLinearInterpolate
   inputs [
    A {from=MaterialExpressionMaterialFunctionCall_0, output=Result}
    B {from=MaterialExpressionNamedRerouteUsage_42}
    Alpha {from=MaterialExpressionSaturate_22}
   ]
  }
  MaterialExpressionLinearInterpolate_8 {
   class=MaterialExpressionLinearInterpolate
   inputs [
    A {from=MaterialExpressionLinearInterpolate_1}
    B {from=MaterialExpressionStaticSwitchParameter_12}
    Alpha {from=MaterialExpressionSaturate_7}
   ]
   properties {constAlpha=1}
  }
  MaterialExpressionLinearInterpolate_9 {
   class=MaterialExpressionLinearInterpolate
   inputs [
    A {from=""}
    B {from=""}
    Alpha {from=MaterialExpressionSaturate_8}
   ]
   properties {constA=0.1}
  }
  MaterialExpressionMaterialFunctionCall_0 {
   class=MaterialExpressionMaterialFunctionCall
   inputs [
    Blend {from=MaterialExpressionNamedRerouteUsage_35}
    Base {from=MaterialExpressionStaticSwitchParameter_7}
   ]
   properties {materialFunction=/Engine/Functions/Engine_MaterialFunctions03/Blends/Blend_Screen}
  }
  MaterialExpressionMaterialFunctionCall_1 {
   class=MaterialExpressionMaterialFunctionCall
   inputs [
    Blend {from=MaterialExpressionTextureSample_8, output=R}
    Base {from=MaterialExpressionTextureSample_8, output=R}
   ]
   properties {materialFunction=&Blend_Screen}
  }
  MaterialExpressionMaterialFunctionCall_10 {
   class=MaterialExpressionMaterialFunctionCall
   inputs [
    In {from=MaterialExpressionAdd_16}
    Contrast {from=MaterialExpressionScalarParameter_1}
   ]
   properties {materialFunction=&CheapContrast}
  }
  MaterialExpressionMaterialFunctionCall_11 {
   class=MaterialExpressionMaterialFunctionCall
   inputs [
    SceneDepth {from=""}
    Exclude WPO Offsets {from=""}
   ]
   properties {materialFunction=&WorldPositionBehindTranslucency}
  }
  MaterialExpressionMaterialFunctionCall_12 {
   class=MaterialExpressionMaterialFunctionCall
   inputs [
    SDF {from=MaterialExpressionPower_7}
    "Size (Max 0.9)" {from=MaterialExpressionScalarParameter_43}
   ]
   properties {materialFunction=&/MF_SDFSize}
  }
  MaterialExpressionMaterialFunctionCall_13 {
   class=MaterialExpressionMaterialFunctionCall
   inputs [
    "Xb1,PS4" {from=MaterialExpressionMultiply_9}
    "PC-High&&Epic,XSX,PS5" {from=MaterialExpressionMultiply_9}
    "PC-Perf,Android" {from=MaterialExpressionMultiply_57}
    Switch {from=MaterialExpressionMultiply_57}
    "PC-Low(NotPerf)&&Medium" {from=MaterialExpressionMultiply_57}
   ]
   properties {materialFunction=&PlatformSwitch_1D}
  }
  MaterialExpressionMaterialFunctionCall_14 {
   class=MaterialExpressionMaterialFunctionCall
   inputs [
    "Xb1,PS4" {from=MaterialExpressionScalarParameter_8}
    "PC-High&&Epic,XSX,PS5" {from=MaterialExpressionScalarParameter_8}
    "PC-Perf,Android" {from=MaterialExpressionScalarParameter_58}
    Switch {from=MaterialExpressionScalarParameter_58}
    "PC-Low(NotPerf)&&Medium" {from=MaterialExpressionScalarParameter_58}
   ]
   properties {materialFunction=/Game/Effects/Niagara/MasterMaterials/Functions/Common/PlatformSwitch_3D}
  }
  MaterialExpressionMaterialFunctionCall_15 {
   class=MaterialExpressionMaterialFunctionCall
   inputs [
    "Xb1,PS4" {from=MaterialExpressionMultiply_17}
    "PC-High&&Epic,XSX,PS5" {from=MaterialExpressionMultiply_17}
    "PC-Perf,Android" {from=MaterialExpressionConstant_10}
    Switch {from=MaterialExpressionConstant_10}
    "PC-Low(NotPerf)&&Medium" {from=MaterialExpressionConstant_10}
   ]
   properties {materialFunction=&PlatformSwitch_1D}
  }
  MaterialExpressionMaterialFunctionCall_16 {
   class=MaterialExpressionMaterialFunctionCall
   inputs [
    SDF {from=MaterialExpressionMin_0}
    "Size (Max 0.9)" {from=MaterialExpressionScalarParameter_9}
   ]
   properties {materialFunction=&MF_SDFSize}
  }
  MaterialExpressionMaterialFunctionCall_17 {
   class=MaterialExpressionMaterialFunctionCall
   inputs [
    "Xb1,PS4" {from=MaterialExpressionTextureSampleParameter2D_0, output=R}
    "PC-High&&Epic,XSX,PS5" {from=MaterialExpressionTextureSampleParameter2D_0, output=R}
    "PC-Perf,Android" {from=MaterialExpressionMax_2}
    Switch {from=MaterialExpressionMax_2}
    "PC-Low(NotPerf)&&Medium" {from=MaterialExpressionMax_2}
   ]
   properties {materialFunction=&PlatformSwitch_1D}
  }
  MaterialExpressionMaterialFunctionCall_2 {
   class=MaterialExpressionMaterialFunctionCall
   inputs [
    UV Coord {from=MaterialExpressionConvert_1}
    LenghtScale {from=MaterialExpressionConstant_13}
    RadialScale {from=MaterialExpressionConstant_12}
   ]
   properties {materialFunction=/Game/Global/MaterialFunctions/TechArt/MF_PolarCoordinates}
  }
  MaterialExpressionMaterialFunctionCall_22 {
   class=MaterialExpressionMaterialFunctionCall
   inputs [
    "Xb1,PS4" {from=MaterialExpressionReroute_12}
    "PC-High&&Epic,XSX,PS5" {from=MaterialExpressionNamedRerouteUsage_70}
    "PC-Perf,Android" {from=MaterialExpressionReroute_12}
    Switch {from=MaterialExpressionReroute_12}
    "PC-Low(NotPerf)&&Medium" {from=MaterialExpressionReroute_12}
   ]
   properties {materialFunction=&PlatformSwitch_3D}
  }
  MaterialExpressionMaterialFunctionCall_23 {
   class=MaterialExpressionMaterialFunctionCall
   inputs [
    "Xb1,PS4" {from=MaterialExpressionStaticSwitchParameter_10}
    "PC-High&&Epic,XSX,PS5" {from=MaterialExpressionNamedRerouteUsage_69}
    "PC-Perf,Android" {from=MaterialExpressionStaticSwitchParameter_10}
    Switch {from=MaterialExpressionStaticSwitchParameter_10}
    "PC-Low(NotPerf)&&Medium" {from=MaterialExpressionStaticSwitchParameter_10}
   ]
   properties {materialFunction=&PlatformSwitch_3D}
  }
  MaterialExpressionMaterialFunctionCall_24 {
   class=MaterialExpressionMaterialFunctionCall
   inputs [
    Diffuse Texture Object {from=MaterialExpressionTextureObjectParameter_5}
    Spec Texture Object {from=""}
    TextureSize {from=MaterialExpressionScalarParameter_92}
    -------------------------- {from=""}
    Use seperate settings for Z projection {from=""}
    Use seperate texture for Z projection {from=""}
    Texture Size Z projection {from=""}
    Z Plane Diffuse Texture Object {from=""}
    Z Plane Spec Texture Object {from=""}
    "--------------------------      " {from=""}
    ProjectionTransitionContrast {from=MaterialExpressionConstant_30}
    Use Texture Alpha Heightmap to Refine the Transistion {from=MaterialExpressionStaticBool_9}
    HeightLerpTransitionContrast {from=MaterialExpressionConstant_29}
    PickTheDiffuseHeightChannel {from=""}
    TheDiffuseHeightChannel {from=""}
    "--------------------------       " {from=""}
    NormalTextureObject {from=""}
    World Space Normals {from=MaterialExpressionStaticBool_8}
    "--------------------------  " {from=""}
    Export Float 4 Diffuse {from=MaterialExpressionStaticBool_11}
    Export Float 4 Spec {from=""}
    "--------------------------     " {from=""}
    WorldPosition {from=MaterialExpressionAdd_49}
    World Space Normal {from=""}
   ]
   properties {materialFunction=/Engine/Functions/Engine_MaterialFunctions02/Texturing/WorldAlignedTextures_Complex}
  }
  MaterialExpressionMaterialFunctionCall_25 {
   class=MaterialExpressionMaterialFunctionCall
   inputs [
    SceneDepth {from=MaterialExpressionAdd_54}
    Exclude WPO Offsets {from=""}
   ]
   properties {materialFunction=&WorldPositionBehindTranslucency}
  }
  MaterialExpressionMaterialFunctionCall_26 {
   class=MaterialExpressionMaterialFunctionCall
   inputs [
    HeightTexture {from=MaterialExpressionConstant_31}
    TextureCoordinates {from=MaterialExpressionNamedRerouteUsage_1}
    ReferencePlane {from=""}
    HeightRatioInput {from=MaterialExpressionScalarParameter_103}
    CameraVector {from=""}
   ]
   properties {materialFunction=/Engine/Functions/Engine_MaterialFunctions02/Utility/BumpOffset_Advanced.BumpOffset_advanced}
  }
  MaterialExpressionMaterialFunctionCall_27 {
   class=MaterialExpressionMaterialFunctionCall
   inputs [
    UV Coord {from=MaterialExpressionNamedRerouteUsage_68}
    LenghtScale {from=MaterialExpressionConstant_34}
    RadialScale {from=MaterialExpressionConstant_33}
   ]
   properties {materialFunction=&MF_PolarCoordinates}
  }
  MaterialExpressionMaterialFunctionCall_28 {
   class=MaterialExpressionMaterialFunctionCall
   inputs [
    Blend {from=MaterialExpressionTextureSample_10, output=R}
    Base {from=MaterialExpressionTextureSample_10, output=R}
   ]
   properties {materialFunction=&Blend_Screen}
  }
  MaterialExpressionMaterialFunctionCall_3 {
   class=MaterialExpressionMaterialFunctionCall
   inputs [Float3 {from=MaterialExpressionActorPositionWS_0}]
   properties {materialFunction=&BreakOutFloat3Components}
  }
  MaterialExpressionMaterialFunctionCall_4 {
   class=MaterialExpressionMaterialFunctionCall
   inputs [Float2 {from=MaterialExpressionMultiply_32}]
   properties {materialFunction=&BreakOutFloat2Components}
  }
  MaterialExpressionMaterialFunctionCall_5 {
   class=MaterialExpressionMaterialFunctionCall
   inputs [Float2 {from=MaterialExpressionMultiply_22}]
   properties {materialFunction=&BreakOutFloat2Components}
  }
  MaterialExpressionMaterialFunctionCall_6 {
   class=MaterialExpressionMaterialFunctionCall
   inputs [
    SDF {from=MaterialExpressionPower_1}
    "Size (Max 0.9)" {from=MaterialExpressionScalarParameter_50}
   ]
   properties {materialFunction=&MF_SDFSize}
  }
  MaterialExpressionMaterialFunctionCall_7 {
   class=MaterialExpressionMaterialFunctionCall
   inputs [
    SDF {from=MaterialExpressionPower_9}
    "Size (Max 0.9)" {from=MaterialExpressionScalarParameter_45}
   ]
   properties {materialFunction=&MF_SDFSize}
  }
  MaterialExpressionMaterialFunctionCall_8 {
   class=MaterialExpressionMaterialFunctionCall
   inputs [
    Diffuse Texture Object {from=MaterialExpressionTextureObjectParameter_1}
    Spec Texture Object {from=""}
    TextureSize {from=MaterialExpressionScalarParameter_14}
    -------------------------- {from=""}
    Use seperate settings for Z projection {from=""}
    Use seperate texture for Z projection {from=""}
    Texture Size Z projection {from=""}
    Z Plane Diffuse Texture Object {from=""}
    Z Plane Spec Texture Object {from=""}
    "--------------------------      " {from=""}
    ProjectionTransitionContrast {from=MaterialExpressionConstant_15}
    Use Texture Alpha Heightmap to Refine the Transistion {from=MaterialExpressionStaticBool_10}
    HeightLerpTransitionContrast {from=MaterialExpressionConstant_14}
    PickTheDiffuseHeightChannel {from=""}
    TheDiffuseHeightChannel {from=""}
    "--------------------------       " {from=""}
    NormalTextureObject {from=""}
    World Space Normals {from=MaterialExpressionStaticBool_7}
    "--------------------------  " {from=""}
    Export Float 4 Diffuse {from=MaterialExpressionStaticBool_0}
    Export Float 4 Spec {from=""}
    "--------------------------     " {from=""}
    WorldPosition {from=MaterialExpressionAdd_7}
    World Space Normal {from=""}
   ]
   properties {materialFunction=&WorldAlignedTextures_Complex}
  }
  MaterialExpressionMaterialFunctionCall_9 {
   class=MaterialExpressionMaterialFunctionCall
   inputs [
    SDF {from=MaterialExpressionReroute_5}
    "Size (Max 0.9)" {from=MaterialExpressionScalarParameter_80}
   ]
   properties {materialFunction=&MF_SDFSize}
  }
  MaterialExpressionMax_0 {
   class=MaterialExpressionMax
   inputs [
    A {from=MaterialExpressionNamedRerouteUsage_15}
    B {from=MaterialExpressionNamedRerouteDeclaration_24}
   ]
  }
  MaterialExpressionMax_2 {
   class=MaterialExpressionMax
   inputs [
    A {from=MaterialExpressionTextureSampleParameter2D_0, output=R}
    B {from=MaterialExpressionScalarParameter_63}
   ]
  }
  MaterialExpressionMax_5 {
   class=MaterialExpressionMax
   inputs [
    A {from=MaterialExpressionMultiply_80}
    B {from=MaterialExpressionMultiply_46}
   ]
  }
  MaterialExpressionMax_6 {
   class=MaterialExpressionMax
   inputs [
    A {from=MaterialExpressionMultiply_94}
    B {from=MaterialExpressionDivide_7}
   ]
  }
  MaterialExpressionMin_0 {
   class=MaterialExpressionMin
   inputs [
    A {from=MaterialExpressionNamedRerouteDeclaration_16, output=""}
    B {from=MaterialExpressionMin_2, output=""}
   ]
   properties {}
  }
  MaterialExpressionMin_1 {
   class=MaterialExpressionMin
   inputs [
    A {from=MaterialExpressionTextureSample_2, output=G}
    B {from=MaterialExpressionTextureSample_5, output=G}
   ]
   properties {}
  }
  MaterialExpressionMin_2 {
   class=MaterialExpressionMin
   inputs [
    A {from=MaterialExpressionTextureSample_1, output=G}
    B {from=MaterialExpressionMin_1, output=""}
   ]
   properties {}
  }
  MaterialExpressionMultiply_0 {
   class=MaterialExpressionMultiply
   inputs [
    A {from=MaterialExpressionNamedRerouteUsage_51}
    B {from=MaterialExpressionMultiply_3}
   ]
  }
  MaterialExpressionMultiply_1 {
   class=MaterialExpressionMultiply
   inputs [
    A {from=MaterialExpressionMaterialFunctionCall_7, output=Result}
    B {from=MaterialExpressionLinearInterpolate_2}
   ]
  }
  MaterialExpressionMultiply_10 {
   class=MaterialExpressionMultiply
   inputs [
    A {from=MaterialExpressionNamedRerouteUsage_54}
    B {from=MaterialExpressionMultiply_42}
   ]
  }
  MaterialExpressionMultiply_11 {
   class=MaterialExpressionMultiply
   inputs [
    A {from=MaterialExpressionAdd_32}
    B {from=MaterialExpressionScalarParameter_46}
   ]
  }
  MaterialExpressionMultiply_12 {
   class=MaterialExpressionMultiply
   inputs [
    A {from=MaterialExpressionReroute_1}
    B {from=MaterialExpressionSaturate_5}
   ]
   properties {constB=2}
  }
  MaterialExpressionMultiply_13 {
   class=MaterialExpressionMultiply
   inputs [
    A {from=MaterialExpressionOneMinus_2}
    B {from=MaterialExpressionScalarParameter_16}
   ]
  }
  MaterialExpressionMultiply_14 {
   class=MaterialExpressionMultiply
   inputs [
    A {from=MaterialExpressionLinearInterpolate_8}
    B {from=MaterialExpressionQualitySwitch_0}
   ]
  }
  MaterialExpressionMultiply_15 {
   class=MaterialExpressionMultiply
   inputs [
    A {from=MaterialExpressionConvert_0}
    B {from=MaterialExpressionScalarParameter_42}
   ]
  }
  MaterialExpressionMultiply_16 {
   class=MaterialExpressionMultiply
   inputs [
    A {from=MaterialExpressionCameraVectorWS_5}
    B {from=MaterialExpressionDivide_19}
   ]
  }
  MaterialExpressionMultiply_17 {
   class=MaterialExpressionMultiply
   inputs [
    A {from=MaterialExpressionConstantBiasScale_0}
    B {from=MaterialExpressionScalarParameter_2}
   ]
  }
  MaterialExpressionMultiply_18 {
   class=MaterialExpressionMultiply
   inputs [
    A {from=MaterialExpressionOneMinus_0}
    B {from=MaterialExpressionOneMinus_0}
   ]
  }
  MaterialExpressionMultiply_19 {
   class=MaterialExpressionMultiply
   inputs [
    A {from=MaterialExpressionMaterialFunctionCall_12, output=Result}
    B {from=MaterialExpressionSaturate_20}
   ]
  }
  MaterialExpressionMultiply_2 {
   class=MaterialExpressionMultiply
   inputs [
    A {from=MaterialExpressionNamedRerouteUsage_57}
    B {from=MaterialExpressionNamedRerouteUsage_43}
   ]
  }
  MaterialExpressionMultiply_20 {
   class=MaterialExpressionMultiply
   inputs [
    A {from=MaterialExpressionNamedRerouteUsage_10}
    B {from=""}
   ]
   properties {constB=3}
  }
  MaterialExpressionMultiply_21 {
   class=MaterialExpressionMultiply
   inputs [
    A {from=MaterialExpressionOneMinus_4}
    B {from=MaterialExpressionScalarParameter_27}
   ]
  }
  MaterialExpressionMultiply_22 {
   class=MaterialExpressionMultiply
   inputs [
    A {from=MaterialExpressionAdd_1}
    B {from=MaterialExpressionAdd_1}
   ]
  }
  MaterialExpressionMultiply_23 {
   class=MaterialExpressionMultiply
   inputs [
    A {from=MaterialExpressionScalarParameter_0}
    B {from=MaterialExpressionConvert_0}
   ]
  }
  MaterialExpressionMultiply_24 {
   class=MaterialExpressionMultiply
   inputs [
    A {from=MaterialExpressionAdd_6}
    B {from=MaterialExpressionLinearInterpolate_11}
   ]
  }
  MaterialExpressionMultiply_25 {
   class=MaterialExpressionMultiply
   inputs [
    A {from=MaterialExpressionPower_4}
    B {from=""}
   ]
   properties {constB=10}
  }
  MaterialExpressionMultiply_26 {
   class=MaterialExpressionMultiply
   inputs [
    A {from=MaterialExpressionNamedRerouteUsage_38}
    B {from=MaterialExpressionNamedRerouteUsage_34}
   ]
  }
  MaterialExpressionMultiply_27 {
   class=MaterialExpressionMultiply
   inputs [
    A {from=MaterialExpressionScalarParameter_51}
    B {from=MaterialExpressionAdd_34}
   ]
  }
  MaterialExpressionMultiply_28 {
   class=MaterialExpressionMultiply
   inputs [
    A {from=MaterialExpressionPower_2}
    B {from=MaterialExpressionPower_2}
   ]
  }
  MaterialExpressionMultiply_29 {
   class=MaterialExpressionMultiply
   inputs [
    A {from=MaterialExpressionNamedRerouteUsage_46}
    B {from=MaterialExpressionScalarParameter_55}
   ]
  }
  MaterialExpressionMultiply_3 {
   class=MaterialExpressionMultiply
   inputs [
    A {from=MaterialExpressionStaticSwitchParameter_16}
    B {from=MaterialExpressionScalarParameter_11}
   ]
  }
  MaterialExpressionMultiply_30 {
   class=MaterialExpressionMultiply
   inputs [
    A {from=MaterialExpressionSubtract_22}
    B {from=MaterialExpressionScalarParameter_41}
   ]
   properties {constB=0.2}
  }
  MaterialExpressionMultiply_31 {
   class=MaterialExpressionMultiply
   inputs [
    A {from=MaterialExpressionAdd_12}
    B {from=MaterialExpressionScalarParameter_34}
   ]
   properties {constB=90}
  }
  MaterialExpressionMultiply_32 {
   class=MaterialExpressionMultiply
   inputs [
    A {from=MaterialExpressionAdd_17}
    B {from=MaterialExpressionAdd_17}
   ]
  }
  MaterialExpressionMultiply_33 {
   class=MaterialExpressionMultiply
   inputs [
    A {from=MaterialExpressionAdd_18}
    B {from=MaterialExpressionScalarParameter_30}
   ]
   properties {constB=90}
  }
  MaterialExpressionMultiply_34 {
   class=MaterialExpressionMultiply
   inputs [
    A {from=MaterialExpressionScalarParameter_33}
    B {from=""}
   ]
   properties {constB=0.2}
  }
  MaterialExpressionMultiply_35 {
   class=MaterialExpressionMultiply
   inputs [
    A {from=MaterialExpressionNamedRerouteUsage_33}
    B {from=MaterialExpressionNamedRerouteUsage_36}
   ]
  }
  MaterialExpressionMultiply_36 {
   class=MaterialExpressionMultiply
   inputs [
    A {from=MaterialExpressionMultiply_28}
    B {from=MaterialExpressionScalarParameter_38}
   ]
  }
  MaterialExpressionMultiply_37 {
   class=MaterialExpressionMultiply
   inputs [
    A {from=MaterialExpressionNamedRerouteUsage_44}
    B {from=MaterialExpressionScalarParameter_56}
   ]
  }
  MaterialExpressionMultiply_38 {
   class=MaterialExpressionMultiply
   inputs [
    A {from=MaterialExpressionAdd_35}
    B {from=MaterialExpressionScalarParameter_52}
   ]
  }
  MaterialExpressionMultiply_39 {
   class=MaterialExpressionMultiply
   inputs [
    A {from=MaterialExpressionTextureSample_12, output=R}
    B {from=MaterialExpressionVectorParameter_0, output=RGB}
   ]
  }
  MaterialExpressionMultiply_4 {
   class=MaterialExpressionMultiply
   inputs [
    A {from=MaterialExpressionDivide_14}
    B {from=MaterialExpressionWorldPosition_1, output=XYZ}
   ]
  }
  MaterialExpressionMultiply_40 {
   class=MaterialExpressionMultiply
   inputs [
    A {from=MaterialExpressionScalarParameter_20}
    B {from=MaterialExpressionSaturate_1}
   ]
  }
  MaterialExpressionMultiply_41 {
   class=MaterialExpressionMultiply
   inputs [
    A {from=MaterialExpressionMultiply_39}
    B {from=""}
   ]
   properties {constB=5}
  }
  MaterialExpressionMultiply_42 {
   class=MaterialExpressionMultiply
   inputs [
    A {from=MaterialExpressionReroute_8}
    B {from=MaterialExpressionScalarParameter_64}
   ]
  }
  MaterialExpressionMultiply_43 {
   class=MaterialExpressionMultiply
   inputs [
    A {from=MaterialExpressionScalarParameter_61}
    B {from=MaterialExpressionAdd_15}
   ]
  }
  MaterialExpressionMultiply_44 {
   class=MaterialExpressionMultiply
   inputs [
    A {from=MaterialExpressionAdd_20}
    B {from=MaterialExpressionScalarParameter_62}
   ]
  }
  MaterialExpressionMultiply_45 {
   class=MaterialExpressionMultiply
   inputs [
    A {from=MaterialExpressionScalarParameter_23}
    B {from=""}
   ]
   properties {constB=0.2}
  }
  MaterialExpressionMultiply_46 {
   class=MaterialExpressionMultiply
   inputs [
    A {from=MaterialExpressionMultiply_86}
    B {from=MaterialExpressionVectorParameter_2, output=RGB}
   ]
  }
  MaterialExpressionMultiply_47 {
   class=MaterialExpressionMultiply
   inputs [
    A {from=MaterialExpressionNamedRerouteUsage_37}
    B {from=MaterialExpressionMultiply_48}
   ]
  }
  MaterialExpressionMultiply_48 {
   class=MaterialExpressionMultiply
   inputs [
    A {from=MaterialExpressionStaticSwitchParameter_16}
    B {from=MaterialExpressionMaterialFunctionCall_14, output=Result}
   ]
  }
  MaterialExpressionMultiply_49 {
   class=MaterialExpressionMultiply
   inputs [
    A {from=MaterialExpressionMultiply_35}
    B {from=MaterialExpressionNamedRerouteUsage_8}
   ]
   properties {constB=2}
  }
  MaterialExpressionMultiply_5 {
   class=MaterialExpressionMultiply
   inputs [
    A {from=MaterialExpressionEyeAdaptationInverse_1, output=EyeAdaptationInverse}
    B {from=MaterialExpressionScalarParameter_65}
   ]
  }
  MaterialExpressionMultiply_50 {
   class=MaterialExpressionMultiply
   inputs [
    A {from=MaterialExpressionMultiply_49}
    B {from=MaterialExpressionScalarParameter_6}
   ]
   properties {constB=5}
  }
  MaterialExpressionMultiply_51 {
   class=MaterialExpressionMultiply
   inputs [
    A {from=MaterialExpressionNamedRerouteUsage_53}
    B {from=""}
   ]
   properties {constB=3}
  }
  MaterialExpressionMultiply_52 {
   class=MaterialExpressionMultiply
   inputs [
    A {from=MaterialExpressionDistance_2}
    B {from=MaterialExpressionMultiply_30}
   ]
  }
  MaterialExpressionMultiply_53 {
   class=MaterialExpressionMultiply
   inputs [
    A {from=MaterialExpressionTextureSample_12, output=R}
    B {from=""}
   ]
   properties {constB=3}
  }
  MaterialExpressionMultiply_54 {
   class=MaterialExpressionMultiply
   inputs [
    A {from=MaterialExpressionStaticSwitchParameter_18}
    B {from=MaterialExpressionQualitySwitch_1}
   ]
  }
  MaterialExpressionMultiply_55 {
   class=MaterialExpressionMultiply
   inputs [
    A {from=MaterialExpressionSaturate_12}
    B {from=MaterialExpressionVertexColor_0, output=R}
   ]
  }
  MaterialExpressionMultiply_57 {
   class=MaterialExpressionMultiply
   inputs [
    A {from=MaterialExpressionMultiply_9}
    B {from=MaterialExpressionScalarParameter_59}
   ]
  }
  MaterialExpressionMultiply_58 {
   class=MaterialExpressionMultiply
   inputs [
    A {from=MaterialExpressionDistance_2}
    B {from=MaterialExpressionConstant_18}
   ]
   properties {constB=0.5}
  }
  MaterialExpressionMultiply_6 {
   class=MaterialExpressionMultiply
   inputs [
    A {from=MaterialExpressionSaturate_9}
    B {from=MaterialExpressionScalarParameter_37}
   ]
  }
  MaterialExpressionMultiply_61 {
   class=MaterialExpressionMultiply
   inputs [
    A {from=MaterialExpressionDistance_2}
    B {from=MaterialExpressionMaterialFunctionCall_1, output=Result}
   ]
  }
  MaterialExpressionMultiply_62 {
   class=MaterialExpressionMultiply
   inputs [
    A {from=MaterialExpressionSaturate_14}
    B {from=MaterialExpressionNamedRerouteUsage_32}
   ]
  }
  MaterialExpressionMultiply_63 {
   class=MaterialExpressionMultiply
   inputs [
    A {from=MaterialExpressionPower_5}
    B {from=MaterialExpressionNamedRerouteUsage_56}
   ]
  }
  MaterialExpressionMultiply_64 {
   class=MaterialExpressionMultiply
   inputs [
    A {from=MaterialExpressionMultiply_63}
    B {from=MaterialExpressionScalarParameter_3}
   ]
   properties {constB=0.1}
  }
  MaterialExpressionMultiply_65 {
   class=MaterialExpressionMultiply
   inputs [
    A {from=MaterialExpressionMultiply_64}
    B {from=MaterialExpressionSaturate_0}
   ]
  }
  MaterialExpressionMultiply_7 {
   class=MaterialExpressionMultiply
   inputs [
    A {from=MaterialExpressionSaturate_18}
    B {from=MaterialExpressionPower_3}
   ]
  }
  MaterialExpressionMultiply_8 {
   class=MaterialExpressionMultiply
   inputs [
    A {from=MaterialExpressionScalarParameter_39}
    B {from=MaterialExpressionAdd_24}
   ]
  }
  MaterialExpressionMultiply_80 {
   class=MaterialExpressionMultiply
   inputs [
    A {from=MaterialExpressionVectorParameter_5, output=RGB}
    B {from=MaterialExpressionMultiply_82}
   ]
  }
  MaterialExpressionMultiply_82 {
   class=MaterialExpressionMultiply
   inputs [
    A {from=MaterialExpressionTextureSampleParameter2D_3, output=R}
    B {from=MaterialExpressionAdd_50}
   ]
  }
  MaterialExpressionMultiply_83 {
   class=MaterialExpressionMultiply
   inputs [
    A {from=MaterialExpressionOneMinus_9}
    B {from=MaterialExpressionScalarParameter_93}
   ]
  }
  MaterialExpressionMultiply_84 {
   class=MaterialExpressionMultiply
   inputs [
    A {from=MaterialExpressionOneMinus_10}
    B {from=MaterialExpressionScalarParameter_98}
   ]
  }
  MaterialExpressionMultiply_85 {
   class=MaterialExpressionMultiply
   inputs [
    A {from=MaterialExpressionDivide_6}
    B {from=MaterialExpressionWorldPosition_2, output=XYZ}
   ]
  }
  MaterialExpressionMultiply_86 {
   class=MaterialExpressionMultiply
   inputs [
    A {from=MaterialExpressionScalarParameter_99}
    B {from=MaterialExpressionNamedRerouteUsage_52}
   ]
  }
  MaterialExpressionMultiply_87 {
   class=MaterialExpressionMultiply
   inputs [
    A {from=MaterialExpressionConstantBiasScale_3}
    B {from=MaterialExpressionScalarParameter_102}
   ]
  }
  MaterialExpressionMultiply_88 {
   class=MaterialExpressionMultiply
   inputs [
    A {from=MaterialExpressionConstantBiasScale_4}
    B {from=MaterialExpressionScalarParameter_104}
   ]
  }
  MaterialExpressionMultiply_89 {
   class=MaterialExpressionMultiply
   inputs [
    A {from=MaterialExpressionNamedRerouteUsage_65}
    B {from=MaterialExpressionScalarParameter_105}
   ]
  }
  MaterialExpressionMultiply_9 {
   class=MaterialExpressionMultiply
   inputs [
    A {from=MaterialExpressionMaterialFunctionCall_17, output=Result}
    B {from=MaterialExpressionAdd_36}
   ]
  }
  MaterialExpressionMultiply_90 {
   class=MaterialExpressionMultiply
   inputs [
    A {from=MaterialExpressionComponentMask_7}
    B {from=MaterialExpressionOneMinus_11}
   ]
  }
  MaterialExpressionMultiply_91 {
   class=MaterialExpressionMultiply
   inputs [
    A {from=MaterialExpressionTime_3}
    B {from=MaterialExpressionScalarParameter_106}
   ]
  }
  MaterialExpressionMultiply_92 {
   class=MaterialExpressionMultiply
   inputs [
    A {from=MaterialExpressionMaterialFunctionCall_26, output=Distortion}
    B {from=MaterialExpressionScalarParameter_107}
   ]
  }
  MaterialExpressionMultiply_93 {
   class=MaterialExpressionMultiply
   inputs [
    A {from=MaterialExpressionMultiply_96}
    B {from=MaterialExpressionNamedRerouteDeclaration_31}
   ]
  }
  MaterialExpressionMultiply_94 {
   class=MaterialExpressionMultiply
   inputs [
    A {from=MaterialExpressionScalarParameter_101}
    B {from=MaterialExpressionNamedRerouteUsage_67}
   ]
  }
  MaterialExpressionMultiply_95 {
   class=MaterialExpressionMultiply
   inputs [
    A {from=MaterialExpressionDistance_3}
    B {from=MaterialExpressionConstant_36}
   ]
   properties {constB=0.5}
  }
  MaterialExpressionMultiply_96 {
   class=MaterialExpressionMultiply
   inputs [
    A {from=MaterialExpressionMax_5}
    B {from=MaterialExpressionSaturate_19}
   ]
  }
  MaterialExpressionNamedRerouteDeclaration_0 {
   class=MaterialExpressionNamedRerouteDeclaration
   inputs [None {from=MaterialExpressionDotProduct_4}]
   properties {name=Dot Vertex Camera, nodeColor=<rgba 1 0.0297 0 1>}
  }
  MaterialExpressionNamedRerouteDeclaration_1 {
   class=MaterialExpressionNamedRerouteDeclaration
   inputs [None {from=MaterialExpressionTransform_8}]
   properties {name=WS TexCord, nodeColor=<rgba 1 0.8152 0 1>}
  }
  MaterialExpressionNamedRerouteDeclaration_10 {
   class=MaterialExpressionNamedRerouteDeclaration
   inputs [None {from=MaterialExpressionTextureObjectParameter_4}]
   properties {name=Texture, nodeColor=<rgba 1 0.5625 0 1>}
  }
  MaterialExpressionNamedRerouteDeclaration_11 {
   class=MaterialExpressionNamedRerouteDeclaration
   inputs [None {from=MaterialExpressionStaticSwitchParameter_6}]
   properties {name=Alpha Mask, nodeColor=<rgba 0 1 0.8359 1>}
  }
  MaterialExpressionNamedRerouteDeclaration_12 {
   class=MaterialExpressionNamedRerouteDeclaration
   inputs [None {from=MaterialExpressionMultiply_52}]
   properties {name=Polar UV Distort, nodeColor=<rgba 0 0.1328 1 1>}
  }
  MaterialExpressionNamedRerouteDeclaration_13 {
   class=MaterialExpressionNamedRerouteDeclaration
   inputs [None {from=MaterialExpressionEyeAdaptationInverse_0, output=EyeAdaptationInverse}]
   properties {name=Emissive Color, nodeColor=<rgba 1 0 0.9375 1>}
  }
  MaterialExpressionNamedRerouteDeclaration_14 {
   class=MaterialExpressionNamedRerouteDeclaration
   inputs [None {from=MaterialExpressionSaturate_20}]
   properties {name=Alpha Circle, nodeColor=<rgba 0.1719 0 1 1>}
  }
  MaterialExpressionNamedRerouteDeclaration_15 {
   class=MaterialExpressionNamedRerouteDeclaration
   inputs [None {from=MaterialExpressionStaticSwitchParameter_9}]
   properties {name=Circle Alpha, nodeColor=<rgba 0.8984 0 1 1>}
  }
  MaterialExpressionNamedRerouteDeclaration_16 {
   class=MaterialExpressionNamedRerouteDeclaration
   inputs [None {from=MaterialExpressionTextureSample_7, output=G}]
   properties {name=Planar Texture, nodeColor=<rgba 0 1 0.5781 1>}
  }
  MaterialExpressionNamedRerouteDeclaration_17 {
   class=MaterialExpressionNamedRerouteDeclaration
   inputs [None {from=MaterialExpressionStaticSwitchParameter_15}]
   properties {name=Polar Noise, nodeColor=<rgba 1 0.6563 0 1>}
  }
  MaterialExpressionNamedRerouteDeclaration_18 {
   class=MaterialExpressionNamedRerouteDeclaration
   inputs [None {from=MaterialExpressionVectorParameter_1, output=RGB}]
   properties {name=Border Color, nodeColor=<rgba 1 0.1172 0 1>}
  }
  MaterialExpressionNamedRerouteDeclaration_19 {
   class=MaterialExpressionNamedRerouteDeclaration
   inputs [None {from=MaterialExpressionVectorParameter_4, output=RGB}]
   properties {name=Background Color, nodeColor=<rgba 0.3564 0 0.1878 1>}
  }
  MaterialExpressionNamedRerouteDeclaration_2 {
   class=MaterialExpressionNamedRerouteDeclaration
   inputs [None {from=MaterialExpressionOneMinus_3}]
   properties {name=Circle Green, nodeColor=<rgba 0.0444 0.4325 1 1>}
  }
  MaterialExpressionNamedRerouteDeclaration_20 {
   class=MaterialExpressionNamedRerouteDeclaration
   inputs [None {from=MaterialExpressionTextureObjectParameter_0}]
   properties {name=Texture Inside, nodeColor=<rgba 1 0.5625 0 1>}
  }
  MaterialExpressionNamedRerouteDeclaration_21 {
   class=MaterialExpressionNamedRerouteDeclaration
   inputs [None {from=MaterialExpressionMultiply_47}]
   properties {name=Polar BG, nodeColor=<rgba 1 0 0.3281 1>}
  }
  MaterialExpressionNamedRerouteDeclaration_22 {
   class=MaterialExpressionNamedRerouteDeclaration
   inputs [None {from=MaterialExpressionMaterialFunctionCall_13, output=Result}]
   properties {name=SurfaceCracksAlpha, nodeColor=<rgba 1 0 0.2812 1>}
  }
  MaterialExpressionNamedRerouteDeclaration_23 {
   class=MaterialExpressionNamedRerouteDeclaration
   inputs [None {from=MaterialExpressionConvert_1}]
   properties {name=Noise Position, nodeColor=<rgba 0.0781 0 1 1>}
  }
  MaterialExpressionNamedRerouteDeclaration_24 {
   class=MaterialExpressionNamedRerouteDeclaration
   inputs [None {from=MaterialExpressionSaturate_16}]
   properties {name=MainCrackAlpha, nodeColor=<rgba 1 0 0.3984 1>}
  }
  MaterialExpressionNamedRerouteDeclaration_25 {
   class=MaterialExpressionNamedRerouteDeclaration
   inputs [None {from=MaterialExpressionMaterialFunctionCall_26, output=Distortion}]
   properties {name=WorldUVs_Offset_Epic, nodeColor=<rgba 0.5937 1 0 1>}
  }
  MaterialExpressionNamedRerouteDeclaration_3 {
   class=MaterialExpressionNamedRerouteDeclaration
   inputs [None {from=MaterialExpressionStaticSwitchParameter_4}]
   properties {name=Circle Red, nodeColor=<rgba 0.0444 0.4325 1 1>}
  }
  MaterialExpressionNamedRerouteDeclaration_30 {
   class=MaterialExpressionNamedRerouteDeclaration
   inputs [None {from=MaterialExpressionComponentMask_11}]
   properties {name=crackheight, nodeColor=<rgba 0 0.1562 1 1>}
  }
  MaterialExpressionNamedRerouteDeclaration_31 {
   class=MaterialExpressionNamedRerouteDeclaration
   inputs [None {from=MaterialExpressionSphereMask_1}]
   properties {name=RadialGradient, nodeColor=<rgba 0 1 0.9531 1>}
  }
  MaterialExpressionNamedRerouteDeclaration_32 {
   class=MaterialExpressionNamedRerouteDeclaration
   inputs [None {from=MaterialExpressionAdd_51}]
   properties {name=WorldUVs_Epic, nodeColor=<rgba 0 0.8828 1 1>}
  }
  MaterialExpressionNamedRerouteDeclaration_33 {
   class=MaterialExpressionNamedRerouteDeclaration
   inputs [None {from=MaterialExpressionScalarParameter_57}]
   properties {name=Opacity_Epic, nodeColor=<rgba 1 0 0.2344 1>}
  }
  MaterialExpressionNamedRerouteDeclaration_34 {
   class=MaterialExpressionNamedRerouteDeclaration
   inputs [None {from=MaterialExpressionMultiply_5}]
   properties {name=Emissive_Epic, nodeColor=<rgba 0 0.5547 1 1>}
  }
  MaterialExpressionNamedRerouteDeclaration_4 {
   class=MaterialExpressionNamedRerouteDeclaration
   inputs [None {from=MaterialExpressionConvert_2}]
   properties {name=Rune UV Pos, nodeColor=<rgba 0 0.9062 1 1>}
  }
  MaterialExpressionNamedRerouteDeclaration_5 {
   class=MaterialExpressionNamedRerouteDeclaration
   inputs [None {from=MaterialExpressionPanner_0}]
   properties {name=Polar UV, nodeColor=<rgba 1 0.0099 0.0258 1>}
  }
  MaterialExpressionNamedRerouteDeclaration_6 {
   class=MaterialExpressionNamedRerouteDeclaration
   inputs [None {from=MaterialExpressionMultiply_1}]
   properties {name=Inter Color, nodeColor=<rgba 1 0.0235 0.8035 1>}
  }
  MaterialExpressionNamedRerouteDeclaration_7 {
   class=MaterialExpressionNamedRerouteDeclaration
   inputs [None {from=MaterialExpressionAppendVector_0}]
   properties {name=UV, nodeColor=<rgba 0.7109 0 1 1>}
  }
  MaterialExpressionNamedRerouteDeclaration_8 {
   class=MaterialExpressionNamedRerouteDeclaration
   inputs [None {from=MaterialExpressionConvert_6}]
   properties {name=Projected UV, nodeColor=<rgba 0.8281 0 1 1>}
  }
  MaterialExpressionNamedRerouteDeclaration_9 {
   class=MaterialExpressionNamedRerouteDeclaration
   inputs [None {from=MaterialExpressionComponentMask_9}]
   properties {name=Planar UV, nodeColor=<rgba 0 1 0.5078 1>}
  }
  MaterialExpressionNamedRerouteUsage_0 {class=MaterialExpressionNamedRerouteUsage}
  MaterialExpressionNamedRerouteUsage_1 {class=MaterialExpressionNamedRerouteUsage}
  MaterialExpressionNamedRerouteUsage_10 {class=MaterialExpressionNamedRerouteUsage}
  MaterialExpressionNamedRerouteUsage_11 {class=MaterialExpressionNamedRerouteUsage}
  MaterialExpressionNamedRerouteUsage_12 {class=MaterialExpressionNamedRerouteUsage}
  MaterialExpressionNamedRerouteUsage_13 {class=MaterialExpressionNamedRerouteUsage}
  MaterialExpressionNamedRerouteUsage_14 {class=MaterialExpressionNamedRerouteUsage}
  MaterialExpressionNamedRerouteUsage_15 {class=MaterialExpressionNamedRerouteUsage}
  MaterialExpressionNamedRerouteUsage_17 {class=MaterialExpressionNamedRerouteUsage}
  MaterialExpressionNamedRerouteUsage_18 {class=MaterialExpressionNamedRerouteUsage}
  MaterialExpressionNamedRerouteUsage_19 {class=MaterialExpressionNamedRerouteUsage}
  MaterialExpressionNamedRerouteUsage_2 {class=MaterialExpressionNamedRerouteUsage}
  MaterialExpressionNamedRerouteUsage_20 {class=MaterialExpressionNamedRerouteUsage}
  MaterialExpressionNamedRerouteUsage_21 {class=MaterialExpressionNamedRerouteUsage}
  MaterialExpressionNamedRerouteUsage_22 {class=MaterialExpressionNamedRerouteUsage}
  MaterialExpressionNamedRerouteUsage_23 {class=MaterialExpressionNamedRerouteUsage}
  MaterialExpressionNamedRerouteUsage_24 {class=MaterialExpressionNamedRerouteUsage}
  MaterialExpressionNamedRerouteUsage_25 {class=MaterialExpressionNamedRerouteUsage}
  MaterialExpressionNamedRerouteUsage_26 {class=MaterialExpressionNamedRerouteUsage}
  MaterialExpressionNamedRerouteUsage_27 {class=MaterialExpressionNamedRerouteUsage}
  MaterialExpressionNamedRerouteUsage_28 {class=MaterialExpressionNamedRerouteUsage}
  MaterialExpressionNamedRerouteUsage_29 {class=MaterialExpressionNamedRerouteUsage}
  MaterialExpressionNamedRerouteUsage_3 {class=MaterialExpressionNamedRerouteUsage}
  MaterialExpressionNamedRerouteUsage_30 {class=MaterialExpressionNamedRerouteUsage}
  MaterialExpressionNamedRerouteUsage_31 {class=MaterialExpressionNamedRerouteUsage}
  MaterialExpressionNamedRerouteUsage_32 {class=MaterialExpressionNamedRerouteUsage}
  MaterialExpressionNamedRerouteUsage_33 {class=MaterialExpressionNamedRerouteUsage}
  MaterialExpressionNamedRerouteUsage_34 {class=MaterialExpressionNamedRerouteUsage}
  MaterialExpressionNamedRerouteUsage_35 {class=MaterialExpressionNamedRerouteUsage}
  MaterialExpressionNamedRerouteUsage_36 {class=MaterialExpressionNamedRerouteUsage}
  MaterialExpressionNamedRerouteUsage_37 {class=MaterialExpressionNamedRerouteUsage}
  MaterialExpressionNamedRerouteUsage_38 {class=MaterialExpressionNamedRerouteUsage}
  MaterialExpressionNamedRerouteUsage_39 {class=MaterialExpressionNamedRerouteUsage}
  MaterialExpressionNamedRerouteUsage_4 {class=MaterialExpressionNamedRerouteUsage}
  MaterialExpressionNamedRerouteUsage_40 {class=MaterialExpressionNamedRerouteUsage}
  MaterialExpressionNamedRerouteUsage_41 {class=MaterialExpressionNamedRerouteUsage}
  MaterialExpressionNamedRerouteUsage_42 {class=MaterialExpressionNamedRerouteUsage}
  MaterialExpressionNamedRerouteUsage_43 {class=MaterialExpressionNamedRerouteUsage}
  MaterialExpressionNamedRerouteUsage_44 {class=MaterialExpressionNamedRerouteUsage}
  MaterialExpressionNamedRerouteUsage_45 {class=MaterialExpressionNamedRerouteUsage}
  MaterialExpressionNamedRerouteUsage_46 {class=MaterialExpressionNamedRerouteUsage}
  MaterialExpressionNamedRerouteUsage_47 {class=MaterialExpressionNamedRerouteUsage}
  MaterialExpressionNamedRerouteUsage_48 {class=MaterialExpressionNamedRerouteUsage}
  MaterialExpressionNamedRerouteUsage_49 {class=MaterialExpressionNamedRerouteUsage}
  MaterialExpressionNamedRerouteUsage_5 {class=MaterialExpressionNamedRerouteUsage}
  MaterialExpressionNamedRerouteUsage_50 {class=MaterialExpressionNamedRerouteUsage}
  MaterialExpressionNamedRerouteUsage_51 {class=MaterialExpressionNamedRerouteUsage}
  MaterialExpressionNamedRerouteUsage_52 {class=MaterialExpressionNamedRerouteUsage}
  MaterialExpressionNamedRerouteUsage_53 {class=MaterialExpressionNamedRerouteUsage}
  MaterialExpressionNamedRerouteUsage_54 {class=MaterialExpressionNamedRerouteUsage}
  MaterialExpressionNamedRerouteUsage_55 {class=MaterialExpressionNamedRerouteUsage}
  MaterialExpressionNamedRerouteUsage_56 {class=MaterialExpressionNamedRerouteUsage}
  MaterialExpressionNamedRerouteUsage_57 {class=MaterialExpressionNamedRerouteUsage}
  MaterialExpressionNamedRerouteUsage_6 {class=MaterialExpressionNamedRerouteUsage}
  MaterialExpressionNamedRerouteUsage_65 {class=MaterialExpressionNamedRerouteUsage}
  MaterialExpressionNamedRerouteUsage_66 {class=MaterialExpressionNamedRerouteUsage}
  MaterialExpressionNamedRerouteUsage_67 {class=MaterialExpressionNamedRerouteUsage}
  MaterialExpressionNamedRerouteUsage_68 {class=MaterialExpressionNamedRerouteUsage}
  MaterialExpressionNamedRerouteUsage_69 {class=MaterialExpressionNamedRerouteUsage}
  MaterialExpressionNamedRerouteUsage_7 {class=MaterialExpressionNamedRerouteUsage}
  MaterialExpressionNamedRerouteUsage_70 {class=MaterialExpressionNamedRerouteUsage}
  MaterialExpressionNamedRerouteUsage_8 {class=MaterialExpressionNamedRerouteUsage}
  MaterialExpressionNamedRerouteUsage_9 {class=MaterialExpressionNamedRerouteUsage}
  MaterialExpressionNaniteReplace_0 {
   class=MaterialExpressionNaniteReplace
   inputs [
    Default {from=MaterialExpressionConstant_0, output=""}
    Nanite {from=MaterialExpressionConvert_4, output=""}
   ]
   properties {}
  }
  MaterialExpressionOneMinus_0 {class=MaterialExpressionOneMinus, inputs [None {from=MaterialExpressionSaturate_13}]}
  MaterialExpressionOneMinus_1 {class=MaterialExpressionOneMinus, inputs [None {from=MaterialExpressionDistance_0}]}
  MaterialExpressionOneMinus_10 {class=MaterialExpressionOneMinus, inputs [None {from=MaterialExpressionSmoothStep_8}]}
  MaterialExpressionOneMinus_11 {class=MaterialExpressionOneMinus, inputs [None {from=MaterialExpressionNamedRerouteUsage_66}]}
  MaterialExpressionOneMinus_12 {class=MaterialExpressionOneMinus, inputs [None {from=MaterialExpressionSaturate_17}]}
  MaterialExpressionOneMinus_2 {class=MaterialExpressionOneMinus, inputs [None {from=MaterialExpressionSmoothStep_0}]}
  MaterialExpressionOneMinus_3 {class=MaterialExpressionOneMinus, inputs [None {from=MaterialExpressionSaturate_12}]}
  MaterialExpressionOneMinus_4 {class=MaterialExpressionOneMinus, inputs [None {from=MaterialExpressionSmoothStep_1}]}
  MaterialExpressionOneMinus_6 {class=MaterialExpressionOneMinus, inputs [None {from=MaterialExpressionNamedRerouteUsage_57}]}
  MaterialExpressionOneMinus_9 {class=MaterialExpressionOneMinus, inputs [None {from=MaterialExpressionSmoothStep_7}]}
  MaterialExpressionPanner_0 {
   class=MaterialExpressionPanner
   inputs [
    Coordinate {from=MaterialExpressionMaterialFunctionCall_2, output=Result}
    Time {from=MaterialExpressionDivide_4}
    Speed {from=""}
   ]
   properties {speedX=0.4}
  }
  MaterialExpressionPanner_2 {
   class=MaterialExpressionPanner
   inputs [
    Coordinate {from=MaterialExpressionMaterialFunctionCall_27, output=Result}
    Time {from=MaterialExpressionDivide_8}
    Speed {from=""}
   ]
   properties {speedX=0.4}
  }
  MaterialExpressionPixelDepth_0 {class=MaterialExpressionPixelDepth, inputs [], properties {}}
  MaterialExpressionPixelDepth_2 {class=MaterialExpressionPixelDepth, inputs [], properties {}}
  MaterialExpressionPower_0 {
   class=MaterialExpressionPower
   inputs [
    Base {from=MaterialExpressionMaterialFunctionCall_9, output=Result}
    Exp {from=MaterialExpressionScalarParameter_4}
   ]
  }
  MaterialExpressionPower_1 {
   class=MaterialExpressionPower
   inputs [
    Base {from=MaterialExpressionMin_0}
    Exp {from=MaterialExpressionScalarParameter_29}
   ]
  }
  MaterialExpressionPower_2 {
   class=MaterialExpressionPower
   inputs [
    Base {from=MaterialExpressionOneMinus_1}
    Exp {from=MaterialExpressionScalarParameter_36}
   ]
  }
  MaterialExpressionPower_3 {
   class=MaterialExpressionPower
   inputs [
    Base {from=MaterialExpressionNamedRerouteUsage_0}
    Exp {from=""}
   ]
   properties {constExponent=4}
  }
  MaterialExpressionPower_4 {
   class=MaterialExpressionPower
   inputs [
    Base {from=MaterialExpressionNamedRerouteUsage_14}
    Exp {from=""}
   ]
   properties {constExponent=5}
  }
  MaterialExpressionPower_5 {
   class=MaterialExpressionPower
   inputs [
    Base {from=MaterialExpressionMaterialFunctionCall_16, output=Result}
    Exp {from=MaterialExpressionScalarParameter_10}
   ]
  }
  MaterialExpressionPower_6 {
   class=MaterialExpressionPower
   inputs [
    Base {from=MaterialExpressionNamedRerouteUsage_41}
    Exp {from=MaterialExpressionScalarParameter_28}
   ]
   properties {constExponent=10}
  }
  MaterialExpressionPower_7 {
   class=MaterialExpressionPower
   inputs [
    Base {from=MaterialExpressionTextureSample_6, output=G}
    Exp {from=MaterialExpressionScalarParameter_31}
   ]
  }
  MaterialExpressionPower_8 {
   class=MaterialExpressionPower
   inputs [
    Base {from=MaterialExpressionTextureSample_0, output=G}
    Exp {from=MaterialExpressionScalarParameter_32}
   ]
  }
  MaterialExpressionPower_9 {
   class=MaterialExpressionPower
   inputs [
    Base {from=MaterialExpressionNamedRerouteUsage_3}
    Exp {from=MaterialExpressionScalarParameter_35}
   ]
  }
  MaterialExpressionQualitySwitch_0 {
   class=MaterialExpressionQualitySwitch
   inputs [
    Default {from=MaterialExpressionScalarParameter_15, output=""}
    Low {from=MaterialExpressionScalarParameter_13, output=""}
    High {from=MaterialExpressionScalarParameter_15, output=""}
    Medium {from=MaterialExpressionScalarParameter_13, output=""}
    Epic {from=MaterialExpressionScalarParameter_15, output=""}
   ]
   properties {}
  }
  MaterialExpressionQualitySwitch_1 {
   class=MaterialExpressionQualitySwitch
   inputs [
    Default {from=MaterialExpressionScalarParameter_12, output=""}
    Low {from=MaterialExpressionScalarParameter_25, output=""}
    High {from=MaterialExpressionScalarParameter_12, output=""}
    Medium {from=MaterialExpressionScalarParameter_25, output=""}
    Epic {from=MaterialExpressionScalarParameter_12, output=""}
   ]
   properties {}
  }
  MaterialExpressionReroute_0 {class=MaterialExpressionReroute, inputs [None {from=MaterialExpressionNamedRerouteUsage_32}]}
  MaterialExpressionReroute_1 {class=MaterialExpressionReroute, inputs [None {from=MaterialExpressionMultiply_64}]}
  MaterialExpressionReroute_10 {class=MaterialExpressionReroute, inputs [None {from=MaterialExpressionLinearInterpolate_3}]}
  MaterialExpressionReroute_11 {class=MaterialExpressionReroute, inputs [None {from=MaterialExpressionPower_2}]}
  MaterialExpressionReroute_12 {class=MaterialExpressionReroute, inputs [None {from=MaterialExpressionStaticSwitchParameter_11}]}
  MaterialExpressionReroute_2 {class=MaterialExpressionReroute, inputs [None {from=MaterialExpressionMultiply_65}]}
  MaterialExpressionReroute_3 {class=MaterialExpressionReroute, inputs [None {from=MaterialExpressionAdd_11}]}
  MaterialExpressionReroute_4 {class=MaterialExpressionReroute, inputs [None {from=MaterialExpressionTextureCoordinate_7}]}
  MaterialExpressionReroute_5 {class=MaterialExpressionReroute, inputs [None {from=MaterialExpressionPower_8}]}
  MaterialExpressionReroute_6 {class=MaterialExpressionReroute, inputs [None {from=MaterialExpressionStaticSwitchParameter_16}]}
  MaterialExpressionReroute_7 {class=MaterialExpressionReroute, inputs [None {from=MaterialExpressionSaturate_4}]}
  MaterialExpressionReroute_8 {class=MaterialExpressionReroute, inputs [None {from=MaterialExpressionLinearInterpolate_0}]}
  MaterialExpressionReroute_9 {class=MaterialExpressionReroute, inputs [None {from=MaterialExpressionStaticSwitchParameter_7}]}
  MaterialExpressionSaturate_0 {class=MaterialExpressionSaturate, inputs [None {from=MaterialExpressionPower_6}]}
  MaterialExpressionSaturate_1 {class=MaterialExpressionSaturate, inputs [None {from=MaterialExpressionMultiply_14}]}
  MaterialExpressionSaturate_10 {class=MaterialExpressionSaturate, inputs [None {from=MaterialExpressionMultiply_58}]}
  MaterialExpressionSaturate_11 {class=MaterialExpressionSaturate, inputs [None {from=MaterialExpressionPower_0}]}
  MaterialExpressionSaturate_12 {class=MaterialExpressionSaturate, inputs [None {from=MaterialExpressionAdd_19}]}
  MaterialExpressionSaturate_13 {class=MaterialExpressionSaturate, inputs [None {from=MaterialExpressionMaterialFunctionCall_10, output=Result}]}
  MaterialExpressionSaturate_14 {class=MaterialExpressionSaturate, inputs [None {from=MaterialExpressionAdd_30}]}
  MaterialExpressionSaturate_15 {class=MaterialExpressionSaturate, inputs [None {from=MaterialExpressionLinearInterpolate_12}]}
  MaterialExpressionSaturate_16 {class=MaterialExpressionSaturate, inputs [None {from=MaterialExpressionMaterialFunctionCall_12, output=Result}]}
  MaterialExpressionSaturate_17 {class=MaterialExpressionSaturate, inputs [None {from=MaterialExpressionMultiply_95}]}
  MaterialExpressionSaturate_18 {class=MaterialExpressionSaturate, inputs [None {from=MaterialExpressionMaterialFunctionCall_6, output=Result}]}
  MaterialExpressionSaturate_19 {class=MaterialExpressionSaturate, inputs [None {from=MaterialExpressionAdd_58}]}
  MaterialExpressionSaturate_2 {class=MaterialExpressionSaturate, inputs [None {from=MaterialExpressionNamedRerouteUsage_12}]}
  MaterialExpressionSaturate_20 {class=MaterialExpressionSaturate, inputs [None {from=MaterialExpressionAdd_28}]}
  MaterialExpressionSaturate_21 {class=MaterialExpressionSaturate, inputs [None {from=MaterialExpressionMultiply_53}]}
  MaterialExpressionSaturate_22 {class=MaterialExpressionSaturate, inputs [None {from=MaterialExpressionMultiply_2}]}
  MaterialExpressionSaturate_23 {class=MaterialExpressionSaturate, inputs [None {from=MaterialExpressionMultiply_7}]}
  MaterialExpressionSaturate_3 {class=MaterialExpressionSaturate, inputs [None {from=MaterialExpressionNamedRerouteUsage_5}]}
  MaterialExpressionSaturate_4 {class=MaterialExpressionSaturate, inputs [None {from=MaterialExpressionScalarParameter_11}]}
  MaterialExpressionSaturate_5 {class=MaterialExpressionSaturate, inputs [None {from=MaterialExpressionMultiply_20}]}
  MaterialExpressionSaturate_6 {class=MaterialExpressionSaturate, inputs [None {from=MaterialExpressionNamedRerouteUsage_4}]}
  MaterialExpressionSaturate_7 {class=MaterialExpressionSaturate, inputs [None {from=MaterialExpressionScalarParameter_40}]}
  MaterialExpressionSaturate_8 {class=MaterialExpressionSaturate, inputs [None {from=MaterialExpressionMultiply_61}]}
  MaterialExpressionSaturate_9 {class=MaterialExpressionSaturate, inputs [None {from=MaterialExpressionAdd_26}]}
  MaterialExpressionScalarParameter_0 {class=MaterialExpressionScalarParameter, properties {defaultValue=1, group=Back Ground Noise, parameterName=Polar Size X}}
  MaterialExpressionScalarParameter_1 {class=MaterialExpressionScalarParameter, properties {defaultValue=2, group=Circular Gradient, parameterName=Circle Red Contrast}}
  MaterialExpressionScalarParameter_10 {class=MaterialExpressionScalarParameter, properties {defaultValue=52, group=Stone Walls, parameterName=Border Stone Highlight gradient}}
  MaterialExpressionScalarParameter_100 {class=MaterialExpressionScalarParameter, properties {defaultValue=8, group=SurfaceCracks, parameterName=RayMarchSteps}}
  MaterialExpressionScalarParameter_101 {class=MaterialExpressionScalarParameter, properties {defaultValue=0.1, group=SurfaceCracks, parameterName=RayMarchStepDist}}
  MaterialExpressionScalarParameter_102 {class=MaterialExpressionScalarParameter, properties {defaultValue=0.001, parameterName=distort}}
  MaterialExpressionScalarParameter_103 {class=MaterialExpressionScalarParameter, properties {defaultValue=0.015, parameterName=depth size}}
  MaterialExpressionScalarParameter_104 {class=MaterialExpressionScalarParameter, properties {defaultValue=0.05, group=Scalability, parameterName=DitherStrength_Epic}}
  MaterialExpressionScalarParameter_105 {class=MaterialExpressionScalarParameter, properties {defaultValue=20, group=SurfaceCracks, parameterName=RayDepthSmallCracks}}
  MaterialExpressionScalarParameter_106 {class=MaterialExpressionScalarParameter, properties {defaultValue=0.01, parameterName=DistortSpeed}}
  MaterialExpressionScalarParameter_107 {class=MaterialExpressionScalarParameter, properties {defaultValue=3, parameterName=DistortionDepth}}
  MaterialExpressionScalarParameter_108 {class=MaterialExpressionScalarParameter, properties {defaultValue=0.43, parameterName=RadialGradientSize}}
  MaterialExpressionScalarParameter_109 {class=MaterialExpressionScalarParameter, properties {parameterName=RadialGradientHardness}}
  MaterialExpressionScalarParameter_11 {class=MaterialExpressionScalarParameter, properties {defaultValue=1, group=Main Parameters, parameterName=Noise Emissive Intensity}}
  MaterialExpressionScalarParameter_110 {class=MaterialExpressionScalarParameter, properties {defaultValue=8, group=Back Ground Noise, parameterName=NoiseSpeed_Epic}}
  MaterialExpressionScalarParameter_111 {class=MaterialExpressionScalarParameter, properties {parameterName=PolarNoise_Min}}
  MaterialExpressionScalarParameter_112 {class=MaterialExpressionScalarParameter, properties {defaultValue=0.4, parameterName=PolarNoise_Max}}
  MaterialExpressionScalarParameter_12 {class=MaterialExpressionScalarParameter, properties {defaultValue=1, group=Scalability, parameterName=High Quality Emissive - Stone Border}}
  MaterialExpressionScalarParameter_13 {class=MaterialExpressionScalarParameter, properties {defaultValue=2, group=Scalability, parameterName=Quality Emissive - Low}}
  MaterialExpressionScalarParameter_14 {class=MaterialExpressionScalarParameter, properties {defaultValue=512, group=SurfaceCracks, parameterName=SurfaceCracks_Texture_Size}}
  MaterialExpressionScalarParameter_15 {class=MaterialExpressionScalarParameter, properties {defaultValue=0.3, group=Scalability, parameterName=Quality Emissive - Default}}
  MaterialExpressionScalarParameter_16 {class=MaterialExpressionScalarParameter, properties {defaultValue=2.2, group=SurfaceCracks, parameterName=SurfaceCracks_Intensity}}
  MaterialExpressionScalarParameter_17 {class=MaterialExpressionScalarParameter, properties {bUseCustomPrimitiveData=true, parameterName=Runes Position X}}
  MaterialExpressionScalarParameter_18 {class=MaterialExpressionScalarParameter, properties {parameterName=SurfaceCracks_Min}}
  MaterialExpressionScalarParameter_19 {class=MaterialExpressionScalarParameter, properties {defaultValue=0.75, group=SurfaceCracks, parameterName=SurfaceCracks_Max}}
  MaterialExpressionScalarParameter_2 {class=MaterialExpressionScalarParameter, properties {defaultValue=0.5, group=Scalability, parameterName=DitherStrength}}
  MaterialExpressionScalarParameter_20 {class=MaterialExpressionScalarParameter, properties {defaultValue=1, group=Main Parameters, parameterName=Emissive Strenght}}
  MaterialExpressionScalarParameter_21 {class=MaterialExpressionScalarParameter, properties {defaultValue=2, group=Back Ground Noise, parameterName=Noise Speed}}
  MaterialExpressionScalarParameter_22 {class=MaterialExpressionScalarParameter, properties {group=UVs, parameterName=Runes Position Y}}
  MaterialExpressionScalarParameter_23 {class=MaterialExpressionScalarParameter, properties {defaultValue=-0.4349, group=Circular Gradient, parameterName=Circle Green Smooth}}
  MaterialExpressionScalarParameter_24 {class=MaterialExpressionScalarParameter, properties {group=SurfaceCracks, parameterName=SurfaceCracks_Core_Min}}
  MaterialExpressionScalarParameter_25 {
   class=MaterialExpressionScalarParameter
   properties {defaultValue=34.8337, group=Scalability, parameterName=Low Quality Emissive - Stone Border}
  }
  MaterialExpressionScalarParameter_26 {class=MaterialExpressionScalarParameter, properties {defaultValue=0.25, group=SurfaceCracks, parameterName=SurfaceCracks_Core_Max}}
  MaterialExpressionScalarParameter_27 {class=MaterialExpressionScalarParameter, properties {defaultValue=2, group=SurfaceCracks, parameterName=SurfaceCracks_Core_Intensity}}
  MaterialExpressionScalarParameter_28 {class=MaterialExpressionScalarParameter, properties {defaultValue=10, parameterName=Stone Border Alpha Circle}}
  MaterialExpressionScalarParameter_29 {class=MaterialExpressionScalarParameter, properties {defaultValue=1, parameterName=Crack Thickness}}
  MaterialExpressionScalarParameter_3 {class=MaterialExpressionScalarParameter, properties {defaultValue=0.1, group=Stone Walls, parameterName=Border Stone Highlight Inttensity}}
  MaterialExpressionScalarParameter_30 {class=MaterialExpressionScalarParameter, properties {defaultValue=3.6386, group=Circular Gradient, parameterName=Circle Green Size}}
  MaterialExpressionScalarParameter_31 {class=MaterialExpressionScalarParameter, properties {defaultValue=1, parameterName=Crack Thickness}}
  MaterialExpressionScalarParameter_32 {class=MaterialExpressionScalarParameter, properties {defaultValue=1, parameterName=Crack Thickness}}
  MaterialExpressionScalarParameter_33 {class=MaterialExpressionScalarParameter, properties {defaultValue=4.3123, group=Circular Gradient, parameterName=Circle Red Smooth}}
  MaterialExpressionScalarParameter_34 {class=MaterialExpressionScalarParameter, properties {defaultValue=1, group=Circular Gradient, parameterName=Circle Red Size}}
  MaterialExpressionScalarParameter_35 {class=MaterialExpressionScalarParameter, properties {defaultValue=1, parameterName=Crack Thickness}}
  MaterialExpressionScalarParameter_36 {class=MaterialExpressionScalarParameter, properties {defaultValue=7.9077, group=Cracks Fade out, parameterName=OP Fade Size}}
  MaterialExpressionScalarParameter_37 {class=MaterialExpressionScalarParameter, properties {defaultValue=1, group=Main Parameters, parameterName=Int Gradient}}
  MaterialExpressionScalarParameter_38 {class=MaterialExpressionScalarParameter, properties {defaultValue=388.5626, group=Cracks Fade out, parameterName=OP Fade Contrast}}
  MaterialExpressionScalarParameter_39 {class=MaterialExpressionScalarParameter, properties {defaultValue=0.2, group=UVs, parameterName=Runes Size X}}
  MaterialExpressionScalarParameter_4 {class=MaterialExpressionScalarParameter, properties {defaultValue=0.6732, group=Main Shape, parameterName=Outline Contrast}}
  MaterialExpressionScalarParameter_40 {
   class=MaterialExpressionScalarParameter
   properties {defaultValue=0.7126, group=Circular Gradient, parameterName=Circular Emissive Fade, sliderMax=1}
  }
  MaterialExpressionScalarParameter_41 {class=MaterialExpressionScalarParameter, properties {group=Stone Walls, parameterName=Distortion Amount}}
  MaterialExpressionScalarParameter_42 {class=MaterialExpressionScalarParameter, properties {defaultValue=1, group=Back Ground Noise, parameterName=Polar Size Y}}
  MaterialExpressionScalarParameter_43 {class=MaterialExpressionScalarParameter, properties {defaultValue=0.8403, group=Main Parameters, parameterName=Opacity Contrast Walls}}
  MaterialExpressionScalarParameter_44 {class=MaterialExpressionScalarParameter, properties {group=Back Ground Noise, parameterName=Polar Noise Position Y}}
  MaterialExpressionScalarParameter_45 {class=MaterialExpressionScalarParameter, properties {defaultValue=0.8386, group=Main Shape, parameterName=SDF Size - Main Shape}}
  MaterialExpressionScalarParameter_46 {class=MaterialExpressionScalarParameter, properties {defaultValue=1, group=UVs, parameterName=Runes Size Y}}
  MaterialExpressionScalarParameter_47 {class=MaterialExpressionScalarParameter, properties {bUseCustomPrimitiveData=true, parameterName=Runes Position X}}
  MaterialExpressionScalarParameter_48 {class=MaterialExpressionScalarParameter, properties {group=UVs, parameterName=Runes Position Y}}
  MaterialExpressionScalarParameter_49 {class=MaterialExpressionScalarParameter, properties {defaultValue=-2.2, group=Main Parameters, parameterName=Depth Offset, sliderMin=0.01}}
  MaterialExpressionScalarParameter_5 {class=MaterialExpressionScalarParameter, properties {group=Scalability, parameterName=Nanite Offset}}
  MaterialExpressionScalarParameter_50 {class=MaterialExpressionScalarParameter, properties {defaultValue=0.8403, group=Main Parameters, parameterName=Opacity Contrast Inside}}
  MaterialExpressionScalarParameter_51 {class=MaterialExpressionScalarParameter, properties {defaultValue=0.2, group=UVs, parameterName=Runes Size X}}
  MaterialExpressionScalarParameter_52 {class=MaterialExpressionScalarParameter, properties {defaultValue=1, group=UVs, parameterName=Runes Size Y}}
  MaterialExpressionScalarParameter_53 {class=MaterialExpressionScalarParameter, properties {bUseCustomPrimitiveData=true, parameterName=Runes Position X}}
  MaterialExpressionScalarParameter_54 {class=MaterialExpressionScalarParameter, properties {group=UVs, parameterName=Runes Position Y}}
  MaterialExpressionScalarParameter_55 {class=MaterialExpressionScalarParameter, properties {defaultValue=1, parameterName=Circle 2}}
  MaterialExpressionScalarParameter_56 {class=MaterialExpressionScalarParameter, properties {defaultValue=1, parameterName=Circle}}
  MaterialExpressionScalarParameter_57 {class=MaterialExpressionScalarParameter, properties {defaultValue=0.06, parameterName=Opacity_Epic}}
  MaterialExpressionScalarParameter_58 {class=MaterialExpressionScalarParameter, properties {defaultValue=8, group=BG, parameterName=BG Intensity Outside Perf}}
  MaterialExpressionScalarParameter_59 {class=MaterialExpressionScalarParameter, properties {defaultValue=0.35, parameterName=SurfaceCracks_IntensityFactor_Low}}
  MaterialExpressionScalarParameter_6 {class=MaterialExpressionScalarParameter, properties {defaultValue=5, group=Circular Gradient, parameterName=Circular Emissive Int}}
  MaterialExpressionScalarParameter_60 {class=MaterialExpressionScalarParameter, properties {defaultValue=1, parameterName=ExposureInversion_Epic}}
  MaterialExpressionScalarParameter_61 {class=MaterialExpressionScalarParameter, properties {defaultValue=0.2, group=UVs, parameterName=Runes Size X}}
  MaterialExpressionScalarParameter_62 {class=MaterialExpressionScalarParameter, properties {defaultValue=1, group=UVs, parameterName=Runes Size Y}}
  MaterialExpressionScalarParameter_63 {class=MaterialExpressionScalarParameter, properties {parameterName=SurfaceCracks_MinCoverage_Low}}
  MaterialExpressionScalarParameter_64 {class=MaterialExpressionScalarParameter, properties {defaultValue=1, group=Main Shape, parameterName=Outline intensity}}
  MaterialExpressionScalarParameter_65 {class=MaterialExpressionScalarParameter, properties {defaultValue=20, parameterName=ExposureInversionCompensation_Epic}}
  MaterialExpressionScalarParameter_7 {class=MaterialExpressionScalarParameter, properties {group=Back Ground Noise, parameterName=Polar Noise Position X}}
  MaterialExpressionScalarParameter_8 {class=MaterialExpressionScalarParameter, properties {defaultValue=1, group=BG, parameterName=BG Intensity Outside}}
  MaterialExpressionScalarParameter_80 {class=MaterialExpressionScalarParameter, properties {defaultValue=0.3166, group=Main Shape, parameterName=SDF Outline Size, sliderMax=0.9}}
  MaterialExpressionScalarParameter_9 {class=MaterialExpressionScalarParameter, properties {defaultValue=0.6, group=Stone Walls, parameterName=Border Stone Highlight Size}}
  MaterialExpressionScalarParameter_92 {class=MaterialExpressionScalarParameter, properties {defaultValue=512, group=SurfaceCracks, parameterName=SurfaceCracks_Texture_Size}}
  MaterialExpressionScalarParameter_93 {class=MaterialExpressionScalarParameter, properties {defaultValue=1.8, group=SurfaceCracks, parameterName=SurfaceCracks_Intensity_Epic}}
  MaterialExpressionScalarParameter_94 {class=MaterialExpressionScalarParameter, properties {group=SurfaceCracks, parameterName=SurfaceCracks_Min_Epic}}
  MaterialExpressionScalarParameter_95 {class=MaterialExpressionScalarParameter, properties {defaultValue=0.55, group=SurfaceCracks, parameterName=SurfaceCracks_Max_Epic}}
  MaterialExpressionScalarParameter_96 {class=MaterialExpressionScalarParameter, properties {group=SurfaceCracks, parameterName=SurfaceCracks_Core_Min_Epic}}
  MaterialExpressionScalarParameter_97 {class=MaterialExpressionScalarParameter, properties {defaultValue=0.25, group=SurfaceCracks, parameterName=SurfaceCracks_Core_Max_Epic}}
  MaterialExpressionScalarParameter_98 {class=MaterialExpressionScalarParameter, properties {defaultValue=2, group=SurfaceCracks, parameterName=SurfaceCracks_Core_Intensity_Epic}}
  MaterialExpressionScalarParameter_99 {class=MaterialExpressionScalarParameter, properties {defaultValue=2, parameterName=SurfaceCracks_Top_Intensity_Epic}}
  MaterialExpressionSceneDepth_0 {class=MaterialExpressionSceneDepth, inputs [UVs {from="", output=""}], properties {}}
  MaterialExpressionSceneDepth_4 {class=MaterialExpressionSceneDepth, inputs [UVs {from="", output=""}], properties {}}
  MaterialExpressionSceneDepth_5 {class=MaterialExpressionSceneDepth, inputs [UVs {from="", output=""}], properties {}}
  MaterialExpressionScreenPosition_1 {class=MaterialExpressionScreenPosition, inputs [], properties {}}
  MaterialExpressionScreenPosition_3 {class=MaterialExpressionScreenPosition, inputs [], properties {}}
  MaterialExpressionSmoothStep_0 {
   class=MaterialExpressionSmoothStep
   inputs [
    Min {from=MaterialExpressionScalarParameter_18, output=""}
    Max {from=MaterialExpressionScalarParameter_19, output=""}
    Value {from=MaterialExpressionComponentMask_0, output=""}
   ]
   properties {constMax=0.32, constMin=0.3}
  }
  MaterialExpressionSmoothStep_1 {
   class=MaterialExpressionSmoothStep
   inputs [
    Min {from=MaterialExpressionScalarParameter_24, output=""}
    Max {from=MaterialExpressionScalarParameter_26, output=""}
    Value {from=MaterialExpressionComponentMask_0, output=""}
   ]
   properties {constMax=0.32, constMin=0.3}
  }
  MaterialExpressionSmoothStep_7 {
   class=MaterialExpressionSmoothStep
   inputs [
    Min {from=MaterialExpressionScalarParameter_94, output=""}
    Max {from=MaterialExpressionScalarParameter_95, output=""}
    Value {from=MaterialExpressionMultiply_90, output=""}
   ]
   properties {constMax=0.32, constMin=0.3}
  }
  MaterialExpressionSmoothStep_8 {
   class=MaterialExpressionSmoothStep
   inputs [
    Min {from=MaterialExpressionScalarParameter_96, output=""}
    Max {from=MaterialExpressionScalarParameter_97, output=""}
    Value {from=MaterialExpressionMultiply_90, output=""}
   ]
   properties {constMax=0.32, constMin=0.3}
  }
  MaterialExpressionSmoothStep_9 {
   class=MaterialExpressionSmoothStep
   inputs [
    Min {from=MaterialExpressionScalarParameter_111, output=""}
    Max {from=MaterialExpressionScalarParameter_112, output=""}
    Value {from=MaterialExpressionMaterialFunctionCall_28, output=Result}
   ]
   properties {}
  }
  MaterialExpressionSphereMask_1 {
   class=MaterialExpressionSphereMask
   inputs [
    A {from=MaterialExpressionTextureCoordinate_12, output=""}
    B {from=MaterialExpressionConstant_32, output=""}
    Radius {from=MaterialExpressionScalarParameter_108, output=""}
    Hardness {from=MaterialExpressionScalarParameter_109, output=""}
   ]
   properties {attenuationRadius=0.5}
  }
  MaterialExpressionStaticBool_0 {class=MaterialExpressionStaticBool, properties {value=true}}
  MaterialExpressionStaticBool_10 {class=MaterialExpressionStaticBool}
  MaterialExpressionStaticBool_11 {class=MaterialExpressionStaticBool, properties {value=true}}
  MaterialExpressionStaticBool_7 {class=MaterialExpressionStaticBool}
  MaterialExpressionStaticBool_8 {class=MaterialExpressionStaticBool}
  MaterialExpressionStaticBool_9 {class=MaterialExpressionStaticBool}
  MaterialExpressionStaticSwitchParameter_0 {
   class=MaterialExpressionStaticSwitchParameter
   inputs [
    True {from=MaterialExpressionReroute_4, output=""}
    False {from=MaterialExpressionNamedRerouteDeclaration_4, output=""}
   ]
   properties {group=Switch, parameterName=Crack?}
  }
  MaterialExpressionStaticSwitchParameter_1 {
   class=MaterialExpressionStaticSwitchParameter
   inputs [
    True {from=MaterialExpressionSaturate_23, output=""}
    False {from=MaterialExpressionSaturate_18, output=""}
   ]
   properties {group=Switch, parameterName=Crack?}
  }
  MaterialExpressionStaticSwitchParameter_10 {
   class=MaterialExpressionStaticSwitchParameter
   inputs [
    True {from=MaterialExpressionAdd_5, output=""}
    False {from=MaterialExpressionNamedRerouteUsage_12, output=""}
   ]
   properties {group=Switch, parameterName=Outside Crack?}
  }
  MaterialExpressionStaticSwitchParameter_11 {
   class=MaterialExpressionStaticSwitchParameter
   inputs [
    True {from=MaterialExpressionLinearInterpolate_7, output=""}
    False {from=MaterialExpressionReroute_9, output=""}
   ]
   properties {group=Switch, parameterName=Outside Crack?}
  }
  MaterialExpressionStaticSwitchParameter_12 {
   class=MaterialExpressionStaticSwitchParameter
   inputs [
    True {from=MaterialExpressionLinearInterpolate_1, output=""}
    False {from=MaterialExpressionAdd_29, output=""}
   ]
   properties {group=Switch, parameterName=Outside Crack?}
  }
  MaterialExpressionStaticSwitchParameter_13 {
   class=MaterialExpressionStaticSwitchParameter
   inputs [
    True {from=MaterialExpressionReroute_12, output=""}
    False {from=MaterialExpressionMaterialFunctionCall_22, output=Result}
   ]
   properties {parameterName=ForceNotEpicQuality}
  }
  MaterialExpressionStaticSwitchParameter_14 {
   class=MaterialExpressionStaticSwitchParameter
   inputs [
    True {from=MaterialExpressionTextureCoordinate_3, output=""}
    False {from=MaterialExpressionNamedRerouteUsage_2, output=""}
   ]
   properties {group=Switch, parameterName=Crack?}
  }
  MaterialExpressionStaticSwitchParameter_15 {
   class=MaterialExpressionStaticSwitchParameter
   inputs [
    True {from=MaterialExpressionLinearInterpolate_13, output=""}
    False {from=MaterialExpressionLinearInterpolate_9, output=""}
   ]
   properties {group=Switch, parameterName=Crack?}
  }
  MaterialExpressionStaticSwitchParameter_16 {
   class=MaterialExpressionStaticSwitchParameter
   inputs [
    True {from=MaterialExpressionMultiply_62, output=""}
    False {from=MaterialExpressionReroute_0, output=""}
   ]
   properties {group=Switch, parameterName=Crack?}
  }
  MaterialExpressionStaticSwitchParameter_17 {
   class=MaterialExpressionStaticSwitchParameter
   inputs [
    True {from=MaterialExpressionStaticSwitchParameter_10, output=""}
    False {from=MaterialExpressionMaterialFunctionCall_23, output=Result}
   ]
   properties {parameterName=ForceNotEpicQuality}
  }
  MaterialExpressionStaticSwitchParameter_18 {
   class=MaterialExpressionStaticSwitchParameter
   inputs [
    True {from=MaterialExpressionLinearInterpolate_11, output=""}
    False {from=MaterialExpressionMultiply_24, output=""}
   ]
   properties {group=Switch, parameterName=Outside Crack?}
  }
  MaterialExpressionStaticSwitchParameter_2 {
   class=MaterialExpressionStaticSwitchParameter
   inputs [
    True {from=MaterialExpressionAppendVector_3, output=""}
    False {from=MaterialExpressionStaticSwitchParameter_13, output=""}
   ]
   properties {parameterName=Debug Gradient}
  }
  MaterialExpressionStaticSwitchParameter_3 {
   class=MaterialExpressionStaticSwitchParameter
   inputs [
    True {from=MaterialExpressionMultiply_19, output=""}
    False {from=MaterialExpressionSaturate_16, output=""}
   ]
   properties {group=Switch, parameterName=Crack?}
  }
  MaterialExpressionStaticSwitchParameter_4 {
   class=MaterialExpressionStaticSwitchParameter
   inputs [
    True {from=MaterialExpressionMultiply_6, output=""}
    False {from=MaterialExpressionConstant_1, output=""}
   ]
   properties {group=Switch, parameterName=Crack?}
  }
  MaterialExpressionStaticSwitchParameter_5 {
   class=MaterialExpressionStaticSwitchParameter
   inputs [
    True {from=MaterialExpressionConstant_2, output=""}
    False {from=MaterialExpressionOneMinus_3, output=""}
   ]
   properties {group=Switch, parameterName=Crack?}
  }
  MaterialExpressionStaticSwitchParameter_6 {
   class=MaterialExpressionStaticSwitchParameter
   inputs [
    True {from=MaterialExpressionMax_0, output=""}
    False {from=MaterialExpressionNamedRerouteDeclaration_24, output=""}
   ]
   properties {group=SurfaceCracks, parameterName=UseSurfaceCracks, sortPriority=-100}
  }
  MaterialExpressionStaticSwitchParameter_7 {
   class=MaterialExpressionStaticSwitchParameter
   inputs [
    True {from=MaterialExpressionLinearInterpolate_15, output=""}
    False {from=MaterialExpressionReroute_10, output=""}
   ]
   properties {group=SurfaceCracks, parameterName=UseSurfaceCracks, sortPriority=-100}
  }
  MaterialExpressionStaticSwitchParameter_8 {
   class=MaterialExpressionStaticSwitchParameter
   inputs [
    True {from=MaterialExpressionReroute_2, output=""}
    False {from=MaterialExpressionMultiply_12, output=""}
   ]
   properties {group=Switch, parameterName=Crack?}
  }
  MaterialExpressionStaticSwitchParameter_9 {
   class=MaterialExpressionStaticSwitchParameter
   inputs [
    True {from=MaterialExpressionMultiply_55, output=""}
    False {from=MaterialExpressionConstant_6, output=""}
   ]
   properties {group=Switch, parameterName=Outside Crack?}
  }
  MaterialExpressionSubtract_1 {
   class=MaterialExpressionSubtract
   inputs [
    A {from=MaterialExpressionMultiply_85}
    B {from=MaterialExpressionActorPositionWS_3}
   ]
  }
  MaterialExpressionSubtract_20 {
   class=MaterialExpressionSubtract
   inputs [
    A {from=MaterialExpressionMultiply_4}
    B {from=MaterialExpressionActorPositionWS_1}
   ]
  }
  MaterialExpressionSubtract_22 {
   class=MaterialExpressionSubtract
   inputs [
    A {from=MaterialExpressionMaterialFunctionCall_1, output=Result}
    B {from=""}
   ]
   properties {constB=0.5}
  }
  MaterialExpressionTextureCoordinate_0 {class=MaterialExpressionTextureCoordinate}
  MaterialExpressionTextureCoordinate_1 {class=MaterialExpressionTextureCoordinate}
  MaterialExpressionTextureCoordinate_10 {class=MaterialExpressionTextureCoordinate, properties {uTiling=8, vTiling=8}}
  MaterialExpressionTextureCoordinate_11 {class=MaterialExpressionTextureCoordinate}
  MaterialExpressionTextureCoordinate_12 {class=MaterialExpressionTextureCoordinate}
  MaterialExpressionTextureCoordinate_13 {class=MaterialExpressionTextureCoordinate}
  MaterialExpressionTextureCoordinate_3 {class=MaterialExpressionTextureCoordinate}
  MaterialExpressionTextureCoordinate_5 {class=MaterialExpressionTextureCoordinate}
  MaterialExpressionTextureCoordinate_6 {class=MaterialExpressionTextureCoordinate}
  MaterialExpressionTextureCoordinate_7 {class=MaterialExpressionTextureCoordinate}
  MaterialExpressionTextureObjectParameter_0 {
   class=MaterialExpressionTextureObjectParameter
   inputs [
    UVs {from="", output=""}
    Apply View MipBias {from="", output=""}
   ]
   properties {parameterName=Interior SDF, samplerType=SAMPLERTYPE_LinearGrayscale, texture=&/T_GlyphsInner}
  }
  MaterialExpressionTextureObjectParameter_1 {
   class=MaterialExpressionTextureObjectParameter
   inputs [
    UVs {from="", output=""}
    Apply View MipBias {from="", output=""}
   ]
   properties {group=SurfaceCracks, parameterName=SurfaceCracks_Texture, texture=/Hermes/Sets/MaskMaker/Texture/T_Hermes_MaskMaker_TileFloor_A_A_D}
  }
  MaterialExpressionTextureObjectParameter_4 {
   class=MaterialExpressionTextureObjectParameter
   inputs [
    UVs {from="", output=""}
    Apply View MipBias {from="", output=""}
   ]
   properties {parameterName=Rune SDF, samplerType=SAMPLERTYPE_LinearGrayscale, texture=&/T_GlyphsAlpha}
  }
  MaterialExpressionTextureObjectParameter_5 {
   class=MaterialExpressionTextureObjectParameter
   inputs [
    UVs {from="", output=""}
    Apply View MipBias {from="", output=""}
   ]
   properties {group=SurfaceCracks, parameterName=SurfaceCracks_Texture, texture=&T_Hermes_MaskMaker_TileFloor_A_A_D}
  }
  MaterialExpressionTextureObject_1 {
   class=MaterialExpressionTextureObject
   inputs []
   properties {samplerType=SAMPLERTYPE_LinearGrayscale, texture=&/T_Cracks_Outer_Epic}
  }
  MaterialExpressionTextureSampleParameter2D_0 {
   class=MaterialExpressionTextureSampleParameter2D
   inputs [
    UVs {from="", output=""}
    Apply View MipBias {from="", output=""}
   ]
   properties {
    group=SurfaceCracks
    parameterName=SurfaceCracks_Mask
    samplerType=SAMPLERTYPE_LinearGrayscale
    texture=&/T_Cracks_Outer_Blur
   }
  }
  MaterialExpressionTextureSampleParameter2D_3 {
   class=MaterialExpressionTextureSampleParameter2D
   inputs [
    UVs {from="", output=""}
    Apply View MipBias {from="", output=""}
   ]
   properties {
    group=SurfaceCracks
    parameterName=SurfaceCracks_Mask
    samplerType=SAMPLERTYPE_LinearGrayscale
    texture=&T_Cracks_Outer_Blur
   }
  }
  MaterialExpressionTextureSample_0 {
   class=MaterialExpressionTextureSample
   inputs [
    UVs {from=MaterialExpressionAdd_10}
    Tex {from=MaterialExpressionNamedRerouteUsage_7}
    Apply View MipBias {from=""}
   ]
   properties {samplerType=SAMPLERTYPE_DistanceFieldFont, texture=&T_GlyphsInner}
  }
  MaterialExpressionTextureSample_1 {
   class=MaterialExpressionTextureSample
   inputs [
    UVs {from=MaterialExpressionLinearInterpolate_4}
    Tex {from=MaterialExpressionNamedRerouteUsage_24}
    Apply View MipBias {from=""}
   ]
  }
  MaterialExpressionTextureSample_10 {
   class=MaterialExpressionTextureSample
   inputs [
    UVs {from=MaterialExpressionPanner_2}
    Tex {from=""}
    Apply View MipBias {from=""}
   ]
   properties {texture=&/3d_VecNoise}
  }
  MaterialExpressionTextureSample_12 {
   class=MaterialExpressionTextureSample
   inputs [
    UVs {from=MaterialExpressionNamedRerouteUsage_9}
    Tex {from=""}
    Apply View MipBias {from=""}
   ]
   properties {texture=/Hermes/Sets/MaskMaker/Texture/T_Hermes_MaskMaker_CrackedFloor_A_A_D}
  }
  MaterialExpressionTextureSample_2 {
   class=MaterialExpressionTextureSample
   inputs [
    UVs {from=MaterialExpressionLinearInterpolate_5}
    Tex {from=MaterialExpressionNamedRerouteUsage_25}
    Apply View MipBias {from=""}
   ]
  }
  MaterialExpressionTextureSample_5 {
   class=MaterialExpressionTextureSample
   inputs [
    UVs {from=MaterialExpressionLinearInterpolate_6}
    Tex {from=MaterialExpressionNamedRerouteUsage_26}
    Apply View MipBias {from=""}
   ]
  }
  MaterialExpressionTextureSample_6 {
   class=MaterialExpressionTextureSample
   inputs [
    UVs {from=MaterialExpressionNamedRerouteUsage_28}
    Tex {from=MaterialExpressionNamedRerouteUsage_27}
    Apply View MipBias {from=""}
   ]
  }
  MaterialExpressionTextureSample_7 {
   class=MaterialExpressionTextureSample
   inputs [
    UVs {from=MaterialExpressionNamedRerouteDeclaration_9}
    Tex {from=MaterialExpressionNamedRerouteUsage_23}
    Apply View MipBias {from=""}
   ]
  }
  MaterialExpressionTextureSample_8 {
   class=MaterialExpressionTextureSample
   inputs [
    UVs {from=MaterialExpressionBumpOffset_3}
    Tex {from=""}
    Apply View MipBias {from=""}
   ]
   properties {texture=&3d_VecNoise}
  }
  MaterialExpressionTextureSample_9 {
   class=MaterialExpressionTextureSample
   inputs [
    UVs {from=MaterialExpressionAdd_55}
    Tex {from=""}
    Apply View MipBias {from=""}
   ]
   properties {
    samplerType=SAMPLERTYPE_LinearGrayscale
    texture=/Game/Global/Materials/MaterialGym/Materials/SmokeWithFlowmap/Textures/T_Swirly_Noise_MiddleDetail
   }
  }
  MaterialExpressionTime_0 {class=MaterialExpressionTime}
  MaterialExpressionTime_3 {class=MaterialExpressionTime}
  MaterialExpressionTime_4 {class=MaterialExpressionTime}
  MaterialExpressionTransform_0 {
   class=MaterialExpressionTransform
   inputs [None {from=MaterialExpressionDivide_16}]
   properties {transformSourceType=TRANSFORMSOURCE_World, transformType=TRANSFORM_Local}
  }
  MaterialExpressionTransform_2 {
   class=MaterialExpressionTransform
   inputs [None {from=MaterialExpressionAdd_33}]
   properties {transformSourceType=TRANSFORMSOURCE_World, transformType=TRANSFORM_Tangent}
  }
  MaterialExpressionTransform_3 {
   class=MaterialExpressionTransform
   inputs [None {from=MaterialExpressionDivide_5}]
   properties {transformSourceType=TRANSFORMSOURCE_World, transformType=TRANSFORM_Local}
  }
  MaterialExpressionTransform_4 {
   class=MaterialExpressionTransform
   inputs [None {from=MaterialExpressionDivide_18}]
   properties {transformSourceType=TRANSFORMSOURCE_World, transformType=TRANSFORM_Tangent}
  }
  MaterialExpressionTransform_8 {class=MaterialExpressionTransform, inputs [None {from=MaterialExpressionAppendVector_4}]}
  MaterialExpressionVectorParameter_0 {
   class=MaterialExpressionVectorParameter
   properties {group=Stone Walls, parameterName=Stone Border Color 2, defaultValue=<rgba 0.013 0.0065 0.0119 1>}
  }
  MaterialExpressionVectorParameter_1 {
   class=MaterialExpressionVectorParameter
   properties {group=Main Shape, parameterName=Border color, defaultValue=<rgba 0.4062 0.0212 0.1933 1>}
  }
  MaterialExpressionVectorParameter_2 {class=MaterialExpressionVectorParameter, properties {parameterName=Color2, defaultValue=<rgba 1 0.05 0.5883 1>}}
  MaterialExpressionVectorParameter_3 {
   class=MaterialExpressionVectorParameter
   properties {group=Stone Walls, parameterName=Stone Border Color 2, defaultValue=<rgba 0.013 0.0065 0.0119 1>}
  }
  MaterialExpressionVectorParameter_4 {class=MaterialExpressionVectorParameter, properties {group=Main Shape, parameterName=BG color, defaultValue=<rgba 0.3628 0 0.1922 1>}}
  MaterialExpressionVectorParameter_5 {class=MaterialExpressionVectorParameter, properties {parameterName=Color, defaultValue=<rgba 0.4 0.008 0.1909 1>}}
  MaterialExpressionVectorParameter_6 {
   class=MaterialExpressionVectorParameter
   properties {group=SurfaceCracks, parameterName=WorldOffset, defaultValue=<rgba 767.7555 -256.1653 -1 1>}
  }
  MaterialExpressionVectorParameter_7 {
   class=MaterialExpressionVectorParameter
   properties {group=SurfaceCracks, parameterName=WorldOffset, defaultValue=<rgba 767.7555 -256.1653 -1 1>}
  }
  MaterialExpressionVertexColor_0 {class=MaterialExpressionVertexColor, inputs [], properties {}}
  MaterialExpressionVertexNormalWS_4 {class=MaterialExpressionVertexNormalWS, inputs [], properties {}}
  MaterialExpressionViewProperty_0 {class=MaterialExpressionViewProperty, inputs [], properties {property=MEVP_TemporalSampleIndex}}
  MaterialExpressionViewProperty_2 {class=MaterialExpressionViewProperty, inputs [], properties {property=MEVP_TemporalSampleIndex}}
  MaterialExpressionWorldPosition_1 {class=MaterialExpressionWorldPosition, inputs [], properties {worldPositionShaderOffset=WPT_CameraRelative}}
  MaterialExpressionWorldPosition_2 {class=MaterialExpressionWorldPosition, inputs [], properties {worldPositionShaderOffset=WPT_CameraRelative}}
 }
 properties {
  bAutomaticallySetUsageInEditor=false
  blendMode=BLEND_Translucent
  shadingModel=MSM_Unlit
  translucencyLightingMode=TLM_VolumetricDirectional
  usedShadingModels=Unlit
 }
}
